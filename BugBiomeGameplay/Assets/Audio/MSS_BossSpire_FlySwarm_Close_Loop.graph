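# Quad MetaSound source for the close fly-swarm ambience loop at the Boss Spire, routed to the outdoor ambient sound class.
# On play, a random wave from an array feeds a quad wave player.
# Front and side pairs each pass through an attenuation/spatialization patch driven by distance, azimuth and shared non-spatialized radius variables.
# A Mobile page uses a stereo player and one patch, feeding only the front pair.
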MetaSoundSource /BugBiomeGameplay/Assets/Audio/MSS_BossSpire_FlySwarm_Close_Loop
inputs [
 UE.Attenuation.Distance {type=Float}
 UE.Source.OnPlay {type=Trigger}
 UE.Spatialization.Azimuth {type=Float}
 UE.Spatialization.Elevation {type=Float}
]
interfaces [UE.Attenuation:1.0, UE.OutputFormat.Quad:1.0, UE.Source:1.1, UE.Spatialization:1.0]
outputs [
 "UE.OutputFormat.Quad.Audio:0" {type=Audio}
 "UE.OutputFormat.Quad.Audio:1" {type=Audio}
 "UE.OutputFormat.Quad.Audio:4" {type=Audio}
 "UE.OutputFormat.Quad.Audio:5" {type=Audio}
]
pages [
 Default {
  is_build_graph=true
  input_defaults {UE.Attenuation.Distance=0, UE.Source.OnPlay=null, UE.Spatialization.Azimuth=0, UE.Spatialization.Elevation=0}
  nodes [
   n1 = VariableAccessor.Float(Variable=InitVariable.Float().Variable)
   n2 = VariableAccessor.Float(Variable=InitVariable.Float().Variable)
   ArrayRandomGet_0 = Array.Random Get:WaveAsset:Array(Next=^UE.Source.OnPlay, In Array=VariableAccessor.WaveAsset:Array(Variable=InitVariable.WaveAsset:Array().Variable).Value) {also {defaults {Weights []}}}
   WavePlayer_0 = UE.Wave Player:Quad(Play=@ArrayRandomGet_0.On Next, Wave Asset=@ArrayRandomGet_0.Value) {also {defaults {Loop="<MetaSoundBuilderResult.SUCCEEDED: 0>"}, version [1,1]}}
   MSP_Attenuation_Spatialization_0 = /FortMetasound/Attenuation/MSP_Attenuation_Spatialization(
    Distance=^UE.Attenuation.Distance
    Azimuth=^UE.Spatialization.Azimuth
    NonSpatializedRadiusEnd=@n1.Value
    NonSpatializedRadiusStart=@n2.Value
    Stereo.InL=@WavePlayer_0.Out Side Left
    Stereo.InR=@WavePlayer_0.Out Side Right
   )
   MSP_Attenuation_Spatialization_1 = &MSP_Attenuation_Spatialization(
    Distance=^UE.Attenuation.Distance
    Azimuth=^UE.Spatialization.Azimuth
    NonSpatializedRadiusEnd=@n1.Value
    NonSpatializedRadiusStart=@n2.Value
    Stereo.InL=@WavePlayer_0.Out Front Left
    Stereo.InR=@WavePlayer_0.Out Front Right
   )
  ]
  links [
   {dst="@UE.OutputFormat.Quad.Audio:4", src=MSP_Attenuation_Spatialization_0.Left}
   {dst="@UE.OutputFormat.Quad.Audio:5", src=MSP_Attenuation_Spatialization_0.Right}
   {dst="@UE.OutputFormat.Quad.Audio:0", src=MSP_Attenuation_Spatialization_1.Left}
   {dst="@UE.OutputFormat.Quad.Audio:1", src=MSP_Attenuation_Spatialization_1.Right}
  ]
 }
 Mobile {
  nodes [
   ArrayRandomGet_0 = Array.Random Get:WaveAsset:Array(Next=^UE.Source.OnPlay, In Array=VariableAccessor.WaveAsset:Array(Variable=InitVariable.WaveAsset:Array().Variable).Value) {also {defaults {Weights []}}}
   WavePlayer_0 = UE.Wave Player:Stereo(Play=@ArrayRandomGet_0.On Next, Wave Asset=@ArrayRandomGet_0.Value) {also {defaults {Loop="<MetaSoundBuilderResult.SUCCEEDED: 0>"}, version [1,1]}}
   MSP_Attenuation_Spatialization_0 = &MSP_Attenuation_Spatialization(
    Distance=^UE.Attenuation.Distance
    Azimuth=^UE.Spatialization.Azimuth
    NonSpatializedRadiusEnd=VariableAccessor.Float(Variable=InitVariable.Float().Variable).Value
    NonSpatializedRadiusStart=VariableAccessor.Float(Variable=InitVariable.Float().Variable).Value
    Stereo.InL=@WavePlayer_0.Out Left
    Stereo.InR=@WavePlayer_0.Out Right
   )
  ]
  links [
   {dst="@UE.OutputFormat.Quad.Audio:0", src=MSP_Attenuation_Spatialization_0.Left}
   {dst="@UE.OutputFormat.Quad.Audio:1", src=MSP_Attenuation_Spatialization_0.Right}
  ]
 }
]
properties {
 assetImportData=&self:AssetImportData
 attenuationSettings=&/ATT_BossSpire_FlySwarm_Close
 bLooping=true
 duration=10000
 maxDistance=3650
 numChannels=4
 outputFormat=Quad
 sampleRate=48000
 soundClassObject=&Fort_Ambient_Outdoors_Athena
 volume=0.7
}
