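# Looping mono MetaSound source for the fly swarm ambience at the Boss Spire in BugBiome.
# On play, two wave players take random array picks: one plain, the other pitched down and filtered by distance-based air absorption.
# A mixer weights each layer with its own distance volume curve from a shared wavetable bank, so attenuation is computed in-graph.
# The Mobile page keeps a single unpitched player scaled by the first curve.
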MetaSoundSource /BugBiomeGameplay/Assets/Audio/MSS_BossSpire_FlySwarm_Loop
inputs [
 UE.Attenuation.Distance {type=Float}
 UE.Source.OnPlay {type=Trigger}
]
interfaces [UE.Attenuation:1.0, UE.OutputFormat.Mono:1.0, UE.Source:1.1]
outputs ["UE.OutputFormat.Mono.Audio:0" {type=Audio}]
pages [
 Default {
  is_build_graph=true
  input_defaults {UE.Attenuation.Distance=0, UE.Source.OnPlay=null}
  nodes [
   ArrayRandomGet_1 = Array.Random Get:WaveAsset:Array(Next=^UE.Source.OnPlay, In Array=VariableAccessor.WaveAsset:Array(Variable=InitVariable.WaveAsset:Array().Variable).Value) {also {defaults {Weights []}}}
   WavePlayer_0 = UE.Wave Player:Mono(Play=@ArrayRandomGet_1.On Next, Wave Asset=@ArrayRandomGet_1.Value) {also {defaults {Loop="<MetaSoundBuilderResult.SUCCEEDED: 0>", Pitch Shift=-2}, version [1,1]}}
   ArrayRandomGet_0 = Array.Random Get:WaveAsset:Array(Next=^UE.Source.OnPlay, In Array=VariableAccessor.WaveAsset:Array(Variable=InitVariable.WaveAsset:Array().Variable).Value) {also {defaults {Weights []}}}
   WavePlayer_1 = UE.Wave Player:Mono(Play=@ArrayRandomGet_0.On Next, Wave Asset=@ArrayRandomGet_0.Value) {also {defaults {Loop="<MetaSoundBuilderResult.SUCCEEDED: 0>"}, version [1,1]}}
   AudioMixerAudioMixerMono2_0 = "AudioMixer.Audio Mixer (Mono, 2)"(
    In 1=/FortMetasound/Attenuation/MSP_Attenuation_AirAbsorption(Distance=^UE.Attenuation.Distance, AudioIn=@WavePlayer_0.Out Mono) {also {defaults {LowPassCutoffFrequencyMax=200, MaxDistanceRange=9000}}}.AudioOut
    Gain 0=&MSP_Attenuation_Volume_Curve(Distance=^UE.Attenuation.Distance) {also {defaults {AttenuationCurves=&/WTB_BossSpire_FlySwarm, FalloffDistance=12000, InnerRadius=200}}}.VolumeMultiplier
    Gain 1=&MSP_Attenuation_Volume_Curve(Distance=^UE.Attenuation.Distance) {also {defaults {AttenuationCurves=&WTB_BossSpire_FlySwarm, CurveIndexToEvaluate=1, FalloffDistance=12000}}}.VolumeMultiplier
    In 0=@WavePlayer_1.Out Mono
   )
  ]
  links [{dst="@UE.OutputFormat.Mono.Audio:0", src=AudioMixerAudioMixerMono2_0.Out}]
 }
 Mobile {
  nodes [
   InitVariable.WaveAsset:Array()
   ArrayRandomGet_0 = Array.Random Get:WaveAsset:Array(Next=^UE.Source.OnPlay, In Array=VariableAccessor.WaveAsset:Array(Variable=InitVariable.WaveAsset:Array().Variable).Value) {also {defaults {Weights []}}}
   WavePlayer_0 = UE.Wave Player:Mono(Play=@ArrayRandomGet_0.On Next, Wave Asset=@ArrayRandomGet_0.Value) {also {defaults {Loop="<MetaSoundBuilderResult.SUCCEEDED: 0>"}, version [1,1]}}
   Multiply_0 = UE.Multiply:Audio by Float(
    AdditionalOperands=&MSP_Attenuation_Volume_Curve(Distance=^UE.Attenuation.Distance) {also {defaults {AttenuationCurves=&WTB_BossSpire_FlySwarm, FalloffDistance=12000, InnerRadius=200}}}.VolumeMultiplier
    PrimaryOperand=@WavePlayer_0.Out Mono
   )
  ]
  links [{dst="@UE.OutputFormat.Mono.Audio:0", src=Multiply_0.Out}]
 }
]
properties {
 assetImportData=&self:AssetImportData
 attenuationSettings=&/ATT_BossSpire_FlySwarm_NoAtt
 bLooping=true
 duration=10000
 maxDistance=12000
 sampleRate=48000
 soundClassObject=&Fort_Ambient_Outdoors_Athena
 volume=0.5
}
