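# Looping quad MetaSound source for the boss spire ground swarm ambience, with attenuation computed inside the graph.
# On play it starts two randomly picked waves: a quad layer with short linear falloff and a panned mono layer with curve-based falloff over longer range.
# Distance and azimuth spatialize both; the mono layer mixes into the front pair, while quad sides feed their own outputs.
# The Mobile page keeps one short-range stereo layer on front channels.
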MetaSoundSource /BugBiomeGameplay/Assets/Audio/MSS_BossSpire_GroundSwarm_Loop
inputs [
 UE.Attenuation.Distance {type=Float}
 UE.Source.OnPlay {type=Trigger}
 UE.Spatialization.Azimuth {type=Float}
 UE.Spatialization.Elevation {type=Float}
]
interfaces [UE.Attenuation:1.0, UE.OutputFormat.Quad:1.0, UE.Source:1.1, UE.Spatialization:1.0]
outputs [
 "UE.OutputFormat.Quad.Audio:0" {type=Audio}
 "UE.OutputFormat.Quad.Audio:1" {type=Audio}
 "UE.OutputFormat.Quad.Audio:4" {type=Audio}
 "UE.OutputFormat.Quad.Audio:5" {type=Audio}
]
pages [
 Default {
  is_build_graph=true
  input_defaults {UE.Attenuation.Distance=0, UE.Source.OnPlay=null, UE.Spatialization.Azimuth=0, UE.Spatialization.Elevation=0}
  nodes [
   ArrayRandomGet_0 = Array.Random Get:WaveAsset:Array(Next=^UE.Source.OnPlay, In Array=VariableAccessor.WaveAsset:Array(Variable=InitVariable.WaveAsset:Array().Variable).Value) {also {defaults {Weights []}}}
   WavePlayer_0 = UE.Wave Player:Mono(Play=@ArrayRandomGet_0.On Next, Wave Asset=@ArrayRandomGet_0.Value) {also {defaults {Loop="<MetaSoundBuilderResult.SUCCEEDED: 0>"}, version [1,1]}}
   StereoPanner_0 = UE.Stereo Panner(In=@WavePlayer_0.Out Mono)
   n1 = /FortMetasound/Attenuation/MSP_Attenuation_Spatialization(
    Distance=^UE.Attenuation.Distance
    Azimuth=^UE.Spatialization.Azimuth
    Stereo.InL=@StereoPanner_0.Out Left
    Stereo.InR=@StereoPanner_0.Out Right
   )
   ArrayRandomGet_1 = Array.Random Get:WaveAsset:Array(Next=^UE.Source.OnPlay, In Array=VariableAccessor.WaveAsset:Array(Variable=InitVariable.WaveAsset:Array().Variable).Value)
   WavePlayer_1 = UE.Wave Player:Quad(Play=@ArrayRandomGet_1.On Next, Wave Asset=@ArrayRandomGet_1.Value) {also {defaults {Loop="<MetaSoundBuilderResult.SUCCEEDED: 0>"}, version [1,1]}}
   n2 = &MSP_Attenuation_Spatialization(
    Distance=^UE.Attenuation.Distance
    Azimuth=^UE.Spatialization.Azimuth
    Stereo.InL=@WavePlayer_1.Out Front Left
    Stereo.InR=@WavePlayer_1.Out Front Right
   ) {also {defaults {NonSpatializedRadiusEnd=1000, NonSpatializedRadiusStart=600}}}
   n3 = /FortMetasound/Attenuation/MSP_Attenuation_Volume_Linear(Distance=^UE.Attenuation.Distance) {also {defaults {FalloffDistance=1200, InnerRadius=700}}}
   AudioMixerAudioMixerStereo2_0 = "AudioMixer.Audio Mixer (Stereo, 2)"(
    In 1 L=@n1.Left
    In 1 R=@n1.Right
    In 0 L=@n2.Left
    In 0 R=@n2.Right
    Gain 1=&MSP_Attenuation_Volume_Curve(Distance=^UE.Attenuation.Distance) {also {defaults {AttenuationCurves=&/WTB_BossSpire_GroundSwarm, FalloffDistance=4000}}}.VolumeMultiplier
    Gain 0=@n3.VolumeMultiplier
   )
   n4 = &MSP_Attenuation_Spatialization(
    Distance=^UE.Attenuation.Distance
    Azimuth=^UE.Spatialization.Azimuth
    Stereo.InL=@WavePlayer_1.Out Side Left
    Stereo.InR=@WavePlayer_1.Out Side Right
   ) {also {defaults {NonSpatializedRadiusEnd=1000, NonSpatializedRadiusStart=600}}}
   Multiply_0 = UE.Multiply:Audio by Float(PrimaryOperand=@n4.Left, AdditionalOperands=@n3.VolumeMultiplier)
   Multiply_1 = UE.Multiply:Audio by Float(PrimaryOperand=@n4.Right, AdditionalOperands=@n3.VolumeMultiplier)
  ]
  links [
   {dst="@UE.OutputFormat.Quad.Audio:0", src=AudioMixerAudioMixerStereo2_0.Out L}
   {dst="@UE.OutputFormat.Quad.Audio:1", src=AudioMixerAudioMixerStereo2_0.Out R}
   {dst="@UE.OutputFormat.Quad.Audio:4", src=Multiply_0.Out}
   {dst="@UE.OutputFormat.Quad.Audio:5", src=Multiply_1.Out}
  ]
 }
 Mobile {
  nodes [
   InitVariable.WaveAsset:Array()
   ArrayRandomGet_0 = Array.Random Get:WaveAsset:Array(Next=^UE.Source.OnPlay, In Array=VariableAccessor.WaveAsset:Array(Variable=InitVariable.WaveAsset:Array().Variable).Value)
   WavePlayer_0 = UE.Wave Player:Stereo(Play=@ArrayRandomGet_0.On Next, Wave Asset=@ArrayRandomGet_0.Value) {also {defaults {Loop="<MetaSoundBuilderResult.SUCCEEDED: 0>"}, version [1,1]}}
   n1 = &MSP_Attenuation_Spatialization(Distance=^UE.Attenuation.Distance, Azimuth=^UE.Spatialization.Azimuth, Stereo.InL=@WavePlayer_0.Out Left, Stereo.InR=@WavePlayer_0.Out Right) {also {defaults {NonSpatializedRadiusEnd=1000, NonSpatializedRadiusStart=600}}}
   n2 = &MSP_Attenuation_Volume_Linear(Distance=^UE.Attenuation.Distance) {also {defaults {FalloffDistance=1200, InnerRadius=700}}}
   Multiply_0 = UE.Multiply:Audio by Float(PrimaryOperand=@n1.Left, AdditionalOperands=@n2.VolumeMultiplier)
   Multiply_1 = UE.Multiply:Audio by Float(PrimaryOperand=@n1.Right, AdditionalOperands=@n2.VolumeMultiplier)
  ]
  links [{dst="@UE.OutputFormat.Quad.Audio:0", src=Multiply_0.Out}, {dst="@UE.OutputFormat.Quad.Audio:1", src=Multiply_1.Out}]
 }
]
properties {
 assetImportData=&self:AssetImportData
 attenuationSettings=&/ATT_BossSpire_GroundSwarm_NoAtt
 bLooping=true
 duration=10000
 maxDistance=4000
 numChannels=4
 outputFormat=Quad
 sampleRate=48000
 soundClassObject=&Fort_Ambient_Outdoors_Athena
}
