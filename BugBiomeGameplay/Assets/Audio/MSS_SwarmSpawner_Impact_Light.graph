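# One-shot stereo MetaSound Source for light impact sounds on the bug biome's Swarm Spawner.
# On play, a random get picks one of eight wave variations and feeds a Wave Player, whose completion fires OnFinished.
# An untriggered RandomFloat node supplies a slight upward pitch shift, while the exposed pitch and volume randomization inputs are left unwired.
# It uses dedicated light-impact attenuation, the Fort_Impacts sound class and a concurrency override limiting simultaneous instances.
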MetaSoundSource /BugBiomeGameplay/Assets/Audio/MSS_SwarmSpawner_Impact_Light
inputs [
 Random.Pitch.Base {type=Float}
 Random.Pitch.Spread {type=Float}
 Random.Volume.Random Volume Base {type=Float}
 Random.Volume.Spread {type=Float}
 Sounds {type=WaveAsset:Array}
 UE.Source.OnPlay {type=Trigger}
]
interfaces [UE.OutputFormat.Stereo:1.0, UE.Source.OneShot:1.0, UE.Source:1.1]
outputs [
 "UE.OutputFormat.Stereo.Audio:0" {type=Audio}
 "UE.OutputFormat.Stereo.Audio:1" {type=Audio}
 UE.Source.OneShot.OnFinished {type=Trigger}
]
pages [Default {
  input_defaults {
   Random.Pitch.Base=0
   Random.Pitch.Spread=0
   Random.Volume.Random Volume Base=0
   Random.Volume.Spread=0
   UE.Source.OnPlay=null
   Sounds [
    /BugBiomeGameplay/Assets/Audio/waves/SwarmSpawner_Impact_Light_01
    /BugBiomeGameplay/Assets/Audio/waves/SwarmSpawner_Impact_Light_02
    /BugBiomeGameplay/Assets/Audio/waves/SwarmSpawner_Impact_Light_03
    /BugBiomeGameplay/Assets/Audio/waves/SwarmSpawner_Impact_Light_04
    /BugBiomeGameplay/Assets/Audio/waves/SwarmSpawner_Impact_Light_05
    /BugBiomeGameplay/Assets/Audio/waves/SwarmSpawner_Impact_Light_06
    /BugBiomeGameplay/Assets/Audio/waves/SwarmSpawner_Impact_Light_07
    /BugBiomeGameplay/Assets/Audio/waves/SwarmSpawner_Impact_Light_08
   ]
  }
  nodes [
   ArrayRandomGet_0 = Array.Random Get:WaveAsset:Array(In Array=^Sounds, Next=^UE.Source.OnPlay)
   WavePlayer_0 = UE.Wave Player:Stereo(Play=@ArrayRandomGet_0.On Next, Wave Asset=@ArrayRandomGet_0.Value, Pitch Shift=UE.RandomFloat() {also {defaults {Max=1.5, Min=0.8}}}.Value) {also {version [1,1]}}
  ]
  links [
   {dst="@UE.Source.OneShot.OnFinished", src=WavePlayer_0.On Finished}
   {dst="@UE.OutputFormat.Stereo.Audio:0", src=WavePlayer_0.Out Left}
   {dst="@UE.OutputFormat.Stereo.Audio:1", src=WavePlayer_0.Out Right}
  ]
 }]
properties {
 assetImportData=&self:AssetImportData
 attenuationSettings=&/ATT_SwarmSpawner_ImpactsLight
 bOverrideConcurrency=true
 maxDistance=3400
 numChannels=2
 outputFormat=Stereo
 sampleRate=48000
 soundClassObject=&Fort_Impacts
 volume=0.8
 concurrencyOverrides {maxCount=3}
}
