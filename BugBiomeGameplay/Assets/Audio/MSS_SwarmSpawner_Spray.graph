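# One-shot stereo MetaSound Source for the Bug Biome swarm spawner's spray sound.
# On play, a random get picks one of three spray waves for a stereo wave player with fixed pitch shift, which signals finish.
# The random pitch and volume inputs and a random float node are left unwired.
# It uses spray attenuation, a tight farthest-first concurrency limit, and an ambient sidechain bus send.
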MetaSoundSource /BugBiomeGameplay/Assets/Audio/MSS_SwarmSpawner_Spray
inputs [
 Random.Pitch.Base {type=Float}
 Random.Pitch.Spread {type=Float}
 Random.Volume.Random Volume Base {type=Float}
 Random.Volume.Spread {type=Float}
 Sounds {type=WaveAsset:Array}
 UE.Source.OnPlay {type=Trigger}
]
interfaces [UE.OutputFormat.Stereo:1.0, UE.Source.OneShot:1.0, UE.Source:1.1]
outputs [
 "UE.OutputFormat.Stereo.Audio:0" {type=Audio}
 "UE.OutputFormat.Stereo.Audio:1" {type=Audio}
 UE.Source.OneShot.OnFinished {type=Trigger}
]
pages [Default {
  input_defaults {
   Random.Pitch.Base=0
   Random.Pitch.Spread=0
   Random.Volume.Random Volume Base=0
   Random.Volume.Spread=0
   UE.Source.OnPlay=null
   Sounds [
    /BugBiomeGameplay/Assets/Audio/waves/Swarmer_Spawner_Spray_01_C
    /BugBiomeGameplay/Assets/Audio/waves/Swarmer_Spawner_Spray_02_C
    /BugBiomeGameplay/Assets/Audio/waves/Swarmer_Spawner_Spray_03_C
   ]
  }
  nodes [
   UE.RandomFloat() {also {defaults {Max=2, Min=-5}}}
   ArrayRandomGet_0 = Array.Random Get:WaveAsset:Array(In Array=^Sounds, Next=^UE.Source.OnPlay)
   WavePlayer_0 = UE.Wave Player:Stereo(Play=@ArrayRandomGet_0.On Next, Wave Asset=@ArrayRandomGet_0.Value) {also {defaults {Pitch Shift=1}, version [1,1]}}
  ]
  links [
   {dst="@UE.Source.OneShot.OnFinished", src=WavePlayer_0.On Finished}
   {dst="@UE.OutputFormat.Stereo.Audio:0", src=WavePlayer_0.Out Left}
   {dst="@UE.OutputFormat.Stereo.Audio:1", src=WavePlayer_0.Out Right}
  ]
 }]
properties {
 assetImportData=&self:AssetImportData
 attenuationSettings=&/ATT_SwarmSpawner_Spray
 bOverrideConcurrency=true
 maxDistance=6500
 numChannels=2
 outputFormat=Stereo
 sampleRate=48000
 soundClassObject=&Fort_Impacts
 volume=1.8
 busSends [{audioBus=&AB_Sidechain_FortAmbient_Vol, sendLevel=0.5}]
 concurrencyOverrides {maxCount=2, resolutionRule=StopFarthestThenPreventNew}
}
