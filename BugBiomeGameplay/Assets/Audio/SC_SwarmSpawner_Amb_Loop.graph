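# Sound cue for the looping ambient sound of the bug biome's swarm spawner, classed as a looping loot ambient.
# A Mixer layers two looping waves, the spawner loop and an infested loot sac idle loop, with the spawner layer boosted.
# A root Modulator adds slight upward pitch variation.
# Overridden capsule attenuation uses logarithmic falloff, low-pass filtering, occlusion and listener focus, and concurrency is capped at two instances.
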SoundCue /BugBiomeGameplay/Assets/Audio/SC_SwarmSpawner_Amb_Loop
root=Modulator_0
nodes [
 {id=Mixer_0, children [WavePlayer_1,WavePlayer_0], properties {inputVolume [1,1.8]}}
 {id=Modulator_0, children [Mixer_0], properties {pitchMax=1.1, pitchMin=1}}
 {id=WavePlayer_0, properties {bLooping=true, soundWaveAssetPtr=/BugBiomeGameplay/Assets/Audio/waves/Swarm_Spawner_Amb_Loop_01}}
 {id=WavePlayer_1, properties {bLooping=true, soundWaveAssetPtr=/BugBiomeGameplay/Assets/Audio/waves/Infest_LootSac_Idle_Loop_01}}
]
properties {
 bOverrideAttenuation=true
 bOverrideConcurrency=true
 maxDistance=2100
 soundClassObject=/Game/Sounds/FortSoundClassesAndModes/Fort_Loot_Ambient_Loop
 volumeMultiplier=0.6
 attenuationOverrides {
  attenuationShape=Capsule
  bAttenuateWithLPF=true
  bEnableFocusInterpolation=true
  bEnableListenerFocus=true
  bEnableLogFrequencyScaling=true
  bEnableOcclusion=true
  distanceAlgorithm=Logarithmic
  falloffDistance=1500
  focusAttackInterpSpeed=0.2
  focusAzimuth=20
  focusVolumeAttenuation=1.3
  hPFFrequencyAtMax=5000
  lPFFrequencyAtMin=1000
  lPFRadiusMax=1000
  lPFRadiusMin=400
  nonFocusAzimuth=40
  nonFocusVolumeAttenuation=0.4
  occlusionLowPassFilterFrequency=1000
  occlusionTraceChannel=ECC_GameTraceChannel11
  stereoSpread=100
  attenuationShapeExtents=<600 400>
 }
 concurrencyOverrides {maxCount=2}
}
