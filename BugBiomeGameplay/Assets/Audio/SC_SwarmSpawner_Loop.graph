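# Looping ambient Sound Cue for the bug biome's swarm spawner, layering a fly drone with randomized bug flybys.
# A pitched-down looping flies wave is mixed louder than a looped random pick of three sharply pitched-up flybys, all under slight pitch variation.
# The flyby layer has its own tighter attenuation that turns non-spatialized up close.
# Cue-level overrides add distance filtering, occlusion, listener focus and a two-voice concurrency limit.
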SoundCue /BugBiomeGameplay/Assets/Audio/SC_SwarmSpawner_Loop
root=Modulator_1
nodes [
 {
  id=Attenuation_0
  children [Modulator_0]
  properties {
   bOverrideAttenuation=true
   attenuationOverrides {distanceAlgorithm=Logarithmic, falloffDistance=800, nonSpatializedRadiusEnd=1400, nonSpatializedRadiusMode=Direct2D, stereoSpread=1000}
  }
 }
 {id=Looping_0, children [Modulator_6]}
 {id=Mixer_0, children [Modulator_2, Attenuation_0], properties {inputVolume [2,1]}}
 {id=Modulator_0, children [Looping_0], properties {pitchMax=1.2, pitchMin=1}}
 {id=Modulator_1, children [Mixer_0], properties {pitchMax=1, pitchMin=0.9}}
 {id=Modulator_2, children [WavePlayer_0], properties {pitchMax=0.9, pitchMin=0.9}}
 {id=Modulator_3, children [WavePlayer_1], properties {pitchMax=3, pitchMin=3}}
 {id=Modulator_4, children [WavePlayer_2], properties {pitchMax=2, pitchMin=2}}
 {id=Modulator_5, children [WavePlayer_3], properties {pitchMax=2, pitchMin=2}}
 {id=Modulator_6, children [Random_3], properties {pitchMax=1.1, pitchMin=0.9}}
 {id=Random_3, children [Modulator_3,Modulator_4,Modulator_5], properties {weights [1,1,1]}}
 {id=WavePlayer_0, properties {bLooping=true, soundWaveAssetPtr=/BugBiomeGameplay/Assets/Audio/waves/Swarm_Spawner_Flies_Loop_01}}
 {id=WavePlayer_1, properties {soundWaveAssetPtr=/BugBiomeGameplay/Assets/Audio/waves/Bug_Cloud_Flyby_01}}
 {id=WavePlayer_2, properties {soundWaveAssetPtr=/BugBiomeGameplay/Assets/Audio/waves/Bug_Cloud_Flyby_02}}
 {id=WavePlayer_3, properties {soundWaveAssetPtr=/BugBiomeGameplay/Assets/Audio/waves/Bug_Cloud_Flyby_03}}
]
properties {
 bOverrideAttenuation=true
 bOverrideConcurrency=true
 maxDistance=2400
 soundClassObject=/Game/Sounds/FortSoundClassesAndModes/Fort_Loot_Ambient_Loop
 volumeMultiplier=1.4
 attenuationOverrides {
  bAttenuateWithLPF=true
  bEnableListenerFocus=true
  bEnableLogFrequencyScaling=true
  bEnableOcclusion=true
  distanceAlgorithm=Logarithmic
  falloffDistance=2000
  focusAzimuth=20
  hPFFrequencyAtMax=5000
  lPFRadiusMax=1400
  lPFRadiusMin=400
  nonFocusAzimuth=40
  nonFocusVolumeAttenuation=0.3
  occlusionLowPassFilterFrequency=1000
  occlusionTraceChannel=ECC_GameTraceChannel11
 }
 concurrencyOverrides {maxCount=2}
}
