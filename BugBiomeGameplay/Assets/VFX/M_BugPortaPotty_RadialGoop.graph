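# M_BugPortaPotty_RadialGoop is a masked, lit material for Niagara mesh particles, which by its name draws radial goop for the bug porta-potty effect.
# Particle alpha sets the radius of an inverted sphere mask centered in UV space, so rising alpha erodes the shape from the middle outward.
# That mask multiplies the green channel of a radial reveal texture, while a matching normal map passes through FlattenNormal.
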Material /BugBiomeGameplay/Assets/VFX/M_BugPortaPotty_RadialGoop
graph {
 properties {
  bAutomaticallySetUsageInEditor=false
  bCastRayTracedShadows=false
  bUsedWithNiagaraMeshParticles=true
  blendMode=BLEND_Masked
  usedShadingModels=Default Lit
 }
 nodes [
  MaterialFunctionCall(
   Normal=TextureSample() {samplerType=SAMPLERTYPE_Normal, texture=&/T_RadialReveal_N}.RGB
   Flatness=Constant() {r=3}
  ) {materialFunction=&FlattenNormal}
  n1 = TextureCoordinate()
  n2 = Constant() {r=0.5}
  n3 = ParticleColor()
  Multiply(
   A=OneMinus(None=SphereMask(A=@n1, B=@n2, Radius=@n3.A) {hardnessPercent=0.5})
   B=TextureSample() {samplerType=SAMPLERTYPE_Masks, texture=&/T_RadialReveal_A}.G
  )
 ]
}
