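# Opaque lit material for Niagara mesh particles, named for the bug spawner's bubble effect in the bug biome.
# The green channel of a goo blob mask texture is multiplied by particle color and an Emissive dynamic parameter, so the effect controls tint and glow per particle.
# That result feeds a shared FX emissive adjustment function, alongside a separately sampled goo blob normal map.
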Material /BugBiomeGameplay/Assets/VFX/M_BugSpawner_Bubble
graph {
 properties {bAutomaticallySetUsageInEditor=false, bCastRayTracedShadows=false, bUsedWithNiagaraMeshParticles=true, usedShadingModels=Default Lit}
 nodes [
  n1 = TextureSample() {samplerType=SAMPLERTYPE_Masks, texture=&/T_GooBlob_D}
  n2 = ParticleColor()
  MaterialFunctionCall(Emissive=Multiply(A=Multiply(A=@n1.G, B=@n2.RGB), B=DynamicParameter() {paramNames [Emissive,Param2,Param3,Param4]}.Emissive)) {materialFunction=&MF_FX_EmissiveAdjustments}
  TextureSample() {samplerType=SAMPLERTYPE_Normal, texture=&/T_GooBlob_N}
 ]
}
