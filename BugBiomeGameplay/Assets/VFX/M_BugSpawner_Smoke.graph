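# Translucent, unlit, two-sided smoke trail material for Niagara ribbons, named for the bug spawner effect.
# Color blends from particle color to brightened, panning cell noise through the "Activated" dynamic parameter, then passes through an emissive adjustment function.
# Opacity combines two scrolling smoke trail samples at different tiling, scaled by particle alpha, depth fade and view-angle fade.
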Material /BugBiomeGameplay/Assets/VFX/M_BugSpawner_Smoke
graph {
 properties {
  bAutomaticallySetUsageInEditor=false
  bUsedWithNiagaraRibbons=true
  blendMode=BLEND_Translucent
  shadingModel=MSM_Unlit
  translucencyLightingMode=TLM_VolumetricPerVertexDirectional
  twoSided=true
  usedShadingModels=Unlit
 }
 nodes [
  n1 = ParticleColor()
  n2 = Multiply(A=TextureCoordinate()) {constB=0.25}
  n3 = Multiply(
   A=@n1.RGB
   B=TextureSample(UVs=Panner(Coordinate=@n2) {speedX=-0.2, speedY=0.05}) {samplerType=SAMPLERTYPE_LinearGrayscale, texture=/Game/Effects/Fort_Effects/Textures/Tile/T_CellNoise_Tile_M}.G
  ) {constB=2}
  MaterialFunctionCall(Emissive=LinearInterpolate(A=@n1.RGB, B=Multiply(A=@n3) {constB=1.5}, Alpha=DynamicParameter() {paramNames [Activated,Param2,Param3,Param4]}.Activated)) {materialFunction=&MF_FX_EmissiveAdjustments}
  n4 = TextureCoordinate()
  n5 = ScalarParameter() {defaultValue=1, parameterName=Trail Tile}
  n6 = Constant() {r=1}
  n7 = MaterialFunctionCall(X=@n5, Y=@n6) {materialFunction=&MakeFloat2}
  n8 = MaterialFunctionCall(X=ScalarParameter() {defaultValue=-0.2, parameterName=Trail Speed}, Y=Constant()) {materialFunction=&MakeFloat2}
  n9 = TextureSampleParameter2D(UVs=Panner(Coordinate=Multiply(A=@n4, B=@n7.Result), Speed=@n8.Result) {speedX=-0.3}) {parameterName=Trail Opacity, samplerType=SAMPLERTYPE_Grayscale, texture=/Game/Effects/Fort_Effects/Textures/Smoke/T_Smoke_Trail_Soft_NoClamp}
  n10 = MaterialFunctionCall(X=Multiply(A=@n5) {constB=1.7}, Y=@n6) {materialFunction=&MakeFloat2}
  n11 = TextureSampleParameter2D(UVs=Panner(Coordinate=Multiply(A=@n4, B=@n10.Result), Speed=@n8.Result) {speedX=-0.3}) {parameterName=Trail Opacity, samplerType=SAMPLERTYPE_Grayscale, texture=&T_Smoke_Trail_Soft_NoClamp}
  n12 = Saturate(None=Multiply(A=@n9.G, B=Add(A=@n11.G) {constB=0.35}))
  Multiply(
   A=DepthFade(Opacity=Multiply(A=@n1.A, B=@n12), FadeDistance=ScalarParameter() {defaultValue=100, parameterName=Depth Fade})
   B=MaterialFunctionCall("FadeSensitivity (optional)"=ScalarParameter() {parameterName=Angle Fade}) {materialFunction=&FadeBasedOnViewAngle}.Result
  )
  OneMinus(None=@n9.G)
 ]
}
