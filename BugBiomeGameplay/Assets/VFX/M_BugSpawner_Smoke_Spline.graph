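# Translucent unlit smoke material for bug spawner spline VFX, usable on Niagara ribbons and mesh particles.
# Emissive blends two colors along U, tinted by particle color.
# Opacity combines panning plasma and flame textures, caustic erosion and end masks, then depth fade and camera-angle fade using SplineThicken's normal.
# Dynamic parameters supply width and core UV offset, with Activated blending in extra noise and switching the erosion range.
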Material /BugBiomeGameplay/Assets/VFX/M_BugSpawner_Smoke_Spline
graph {
 properties {
  bUsedWithNiagaraMeshParticles=true
  bUsedWithNiagaraRibbons=true
  blendMode=BLEND_Translucent
  pixelDepthOffsetMode=PDOM_Legacy
  shadingModel=MSM_Unlit
  usedShadingModels=Unlit
 }
 nodes [
  n1 = VectorParameter() {parameterName=Color1, defaultValue=<rgba 5 0.1667 0 1>}
  n2 = VectorParameter() {parameterName=Color2, defaultValue=<rgba 15 1.25 0 1>}
  n3 = TextureCoordinate()
  n4 = Power(Base=OneMinus(None=ComponentMask(None=@n3) {r=true}), Exp=ScalarParameter() {defaultValue=3, parameterName=Lerp Power}) {constExponent=3}
  n5 = ParticleColor()
  MaterialFunctionCall(Emissive=Multiply(A=LinearInterpolate(A=@n1.RGB, B=@n2.RGB, Alpha=@n4), B=@n5.RGB)) {materialFunction=&MF_FX_EmissiveAdjustments}
  n6 = Multiply(
   A=Add(A=TextureCoordinate() {uTiling=2, vTiling=0.9}, B=Constant2Vector() {g=0.05})
   B=AppendVector(A=ScalarParameter() {defaultValue=1, parameterName=UScale}, B=ScalarParameter() {defaultValue=1, parameterName=VScale})
  )
  n7 = ComponentMask(None=VectorParameter() {parameterName=UVscale, defaultValue=<rgba 1 1 0 0>}.RGB) {g=true, r=true}
  n8 = Time()
  n9 = ScalarParameter() {defaultValue=1, parameterName=TimeMult}
  n10 = TextureSampleParameter2D(UVs=Panner(Coordinate=Multiply(A=@n6, B=@n7), Time=Multiply(A=@n8, B=@n9)) {speedX=-0.5}) {parameterName=TexA, samplerType=SAMPLERTYPE_LinearGrayscale, texture=/Game/Effects/Fort_Effects/Emotes/UltraLight/Texture/T_PlasmaCoils}
  n11 = TextureCoordinate()
  n12 = AppendVector(A=DynamicParameter() {defaultValue=<rgba 5 15 0 1>, paramNames [WidthBase,WidthTip,UOffset,Activated]}.UOffset, B=Constant())
  n13 = ScalarParameter() {defaultValue=1, parameterName=UScale}
  n14 = ScalarParameter() {defaultValue=1, parameterName=VScale}
  n15 = Multiply(
   A=Multiply(A=Add(A=@n11, B=@n12), B=AppendVector(A=@n13, B=@n14))
   B=ComponentMask(None=VectorParameter() {parameterName=UVscale, defaultValue=<rgba 1 1 0 0>}.RGB) {g=true, r=true}
  )
  n16 = Multiply(A=Time(), B=ScalarParameter() {defaultValue=1, parameterName=TimeMult})
  n17 = Multiply(
   A=@n10.R
   B=TextureSampleParameter2D(UVs=Panner(Coordinate=@n15, Time=@n16) {speedX=-0.5}) {parameterName=CoreTexture, texture=/Game/Accessories/FORT_Backpacks/Backpack_M_MED_KeyChain/FX/T_Backpack_KeyChain_Flame}.R
  )
  n18 = ScalarParameter() {defaultValue=-0.2, parameterName=ActivatedNoise Speed X}
  n19 = TextureSampleParameter2D(UVs=Panner(
    Coordinate=TextureCoordinate()
    Speed=MaterialFunctionCall(X=@n18) {materialFunction=&MakeFloat2}.Result
   ) {speedX=-0.2}) {
   parameterName=ActivatedNoise
   samplerType=SAMPLERTYPE_LinearGrayscale
   texture=/Engine/Functions/Engine_MaterialFunctions02/ExampleContent/Textures/LowResBlurredNoise
  }
  n20 = Multiply(A=@n19.R, B=@n17)
  n21 = DynamicParameter() {defaultValue=<rgba 5 15 0 1>, paramNames [WidthBase,WidthTip,UOffset,Activated]}
  n22 = Multiply(
   A=TextureCoordinate() {uTiling=0.8}
   B=AppendVector(A=ScalarParameter() {defaultValue=1, parameterName=UScale}, B=ScalarParameter() {defaultValue=1, parameterName=VScale})
  )
  n23 = ComponentMask(None=VectorParameter() {parameterName=UVscale, defaultValue=<rgba 1 1 0 0>}.RGB) {g=true, r=true}
  n24 = LinearInterpolate(
   A=Power(Base=LinearInterpolate(A=@n17, B=@n20, Alpha=@n21.Activated), Exp=ScalarParameter() {defaultValue=1.1, parameterName=Tex power}) {constExponent=1.1}
   B=TextureSampleParameter2D(UVs=Multiply(A=@n22, B=@n23)) {parameterName=CoreTexture, texture=&T_Backpack_KeyChain_Flame}.G
   Alpha=ScalarParameter() {parameterName=Use simple glow mask}
  )
  n25 = ScalarParameter() {defaultValue=2, parameterName=Disappear max deactivated}
  n26 = ScalarParameter() {defaultValue=2, parameterName=Disappear max activated}
  n27 = DynamicParameter() {defaultValue=<rgba 5 15 0 1>, paramNames [WidthBase,WidthTip,UOffset,Activated]}
  n28 = VectorParameter() {parameterName=ErosionTextureSize, defaultValue=<rgba 1 1 0 1>}
  n29 = Add(A=Multiply(A=TextureCoordinate() {uTiling=2}, B=ComponentMask(None=@n28.RGB) {g=true, r=true})) {constB=0}
  n30 = AppendVector(A=ScalarParameter() {defaultValue=1, parameterName=UScale}, B=ScalarParameter() {defaultValue=1, parameterName=VScale})
  n31 = Time()
  n32 = ScalarParameter() {defaultValue=1, parameterName=TimeMult}
  n33 = TextureSampleParameter2D(UVs=Panner(Coordinate=Multiply(A=@n29, B=@n30), Time=Multiply(A=@n31, B=@n32)) {speedX=-0.7}) {
   parameterName=ErosionTexture
   samplerType=SAMPLERTYPE_LinearGrayscale
   texture=/Game/Global/Textures/Default/Caustics/T_Nosie_Caustics_Grayscale
  }
  n34 = ComponentMask(None=TextureCoordinate()) {r=true}
  n35 = Saturate(None=LinearInterpolate(
    A=ScalarParameter() {parameterName=Dissolve mask Min}
    B=ScalarParameter() {defaultValue=0.5, parameterName=Dissolve mask Max}
    Alpha=Power(Base=@n34) {constExponent=1.3}
   ) {constA=0.1, constB=2})
  n36 = Saturate(None=LinearInterpolate(
    A=ScalarParameter() {defaultValue=-2, parameterName=Disappear min}
    B=LinearInterpolate(A=@n25, B=@n26, Alpha=@n27.Activated)
    Alpha=Power(Base=@n33.R, Exp=@n35) {constExponent=0.2}
   ) {constA=-2, constB=2})
  n37 = OneMinus(None=ComponentMask(None=TextureCoordinate()) {r=true})
  n38 = ScalarParameter() {defaultValue=2, parameterName=Tail Power}
  n39 = Saturate(None=Multiply(A=Power(Base=@n37, Exp=@n38) {constExponent=4}, B=ScalarParameter() {defaultValue=8, parameterName=Tail edge hardness}) {constB=5})
  n40 = ComponentMask(None=TextureCoordinate()) {r=true}
  n41 = Saturate(None=Multiply(A=Power(Base=@n40)) {constB=10})
  n42 = Multiply(A=Multiply(A=Multiply(A=@n24, B=@n36), B=Multiply(A=@n39, B=@n41))) {constB=2}
  n43 = DynamicParameter() {defaultValue=<rgba 5 15 0 1>, paramNames [WidthBase,WidthTip,UOffset,Activated]}
  n44 = StaticBoolParameter() {parameterName="Expand U or V UV Channel (SB)"}
  n45 = StaticBool() {value=true}
  n46 = Abs(None=DotProduct(
    A=CameraVectorWS()
    B=MaterialFunctionCall(WidthBase=@n43.WidthBase, WidthTip=@n43.WidthBase, Expand U or V UV Channel=@n44, AngleCorrectedNormals=@n45, FlattenPixelNormal=@n45) {materialFunction=&SplineThicken}.Normal
   ))
  n47 = ScalarParameter() {defaultValue=2, parameterName=AngleFade Power}
  Multiply(
   A=DepthFade(Opacity=Multiply(A=@n5.A, B=@n42), FadeDistance=ScalarParameter() {defaultValue=5, parameterName=DepthFadeDistance}) {fadeDistanceDefault=5}
   B=Saturate(None=Power(Base=@n46, Exp=@n47))
  )
 ]
}
