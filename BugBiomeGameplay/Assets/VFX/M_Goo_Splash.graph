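# Translucent lit, two-sided goo splash material for particle VFX, with purple-tinted default colors.
# Noise-distorted, panning UVs feed a splash mask that Dynamic Parameter Erode thins, with a custom clip and soft edge.
# Opacity also uses particle alpha, near-camera fade and depth fade, which is bypassed on mobile.
# Emissive adds scaled diffuse, a view-space-normal matcap reflection and fake sun/moon specular from Fortnite's parameter collection, with a mobile minimum offset.
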Material /BugBiomeGameplay/Assets/VFX/M_Goo_Splash
graph {
 properties {
  bEnableResponsiveAA=true
  bUsedWithInstancedStaticMeshes=true
  bUsedWithNanite=true
  bUsedWithNiagaraRibbons=true
  bUsedWithNiagaraSprites=true
  bUsedWithParticleSprites=true
  bUsedWithSplineMeshes=true
  blendMode=BLEND_Translucent
  pixelDepthOffsetMode=PDOM_Legacy
  translucentSelfShadowDensityScale=0.1
  translucentShadowDensityScale=0.1
  twoSided=true
  usedShadingModels=Default Lit
 }
 nodes [
  MaterialFunctionCall(Emissive=NamedRerouteUsage()) {materialFunction=&MF_FX_EmissiveAdjustments}
  n1 = NamedRerouteUsage()
  n2 = ScalarParameter() {defaultValue=0.1, parameterName=DiffuseColorEmissiveMultiply}
  n3 = NamedRerouteUsage()
  n4 = Constant() {r=0.5}
  n5 = Add(A=Multiply(A=Transform(None=@n3) {transformType=TRANSFORM_View}, B=@n4), B=@n4)
  n6 = Multiply(
   A=TextureSample(UVs=ComponentMask(None=@n5) {g=true, r=true}) {texture=/Game/VisualThreatManager/StormVisuals/LargeClouds/T_EyeballReflection2}.RGB
   B=ScalarParameter() {defaultValue=1, group=Reflection, parameterName=Reflection Amount}
  )
  n7 = ScalarParameter() {group=Reflection, parameterName=Reflection - Desaturation}
  n8 = ScalarParameter() {
   defaultValue=0.075
   desc="Offset the Minimum fake specular amount for Mobile.  ("
   group=Specular
   parameterName=Mobile Specular Min Offset
   sliderMax=0.2
  }
  n9 = FeatureLevelSwitch(Default=Constant(), ES3_1=@n8)
  n10 = Multiply(
   A=Clamp(None=Desaturation(None=@n6, Fraction=@n7), Min=@n9) {minDefault=0.1}
   B=VectorParameter() {group=Color, parameterName=ReflectionColor, defaultValue=<rgba 0.1911 0.1312 0.5469 1>}.RGB
  )
  n11 = CollectionParameter() {
   collection=&FortniteMaterialParameters
   parameterName=SunAndMoonModelDirectionalVector
  }
  n12 = ComponentMask(None=@n11) {b=true, g=true, r=true}
  n13 = ReflectionVectorWS(CustomWorldNormal=Transform(None=@n3)) {bNormalizeCustomWorldNormal=true}
  n14 = NamedRerouteDeclaration(None=Saturate(None=DotProduct(A=@n12, B=@n13))) {name=Specular, nodeColor=<rgba 0 1 0.4141 1>}
  n15 = ScalarParameter() {defaultValue=10, group=Specular, parameterName=Specular Power}
  n16 = Power(Base=ComponentMask(None=CollectionParameter() {collection=&FortniteMaterialParameters, parameterName=SunLightColor}) {b=true, g=true, r=true}) {constExponent=0.1}
  n17 = Multiply(
   A=Clamp(
    None=Power(Base=@n14, Exp=@n15) {constExponent=40}
    Min=@n9
    Max=ScalarParameter() {defaultValue=5, group=Specular, parameterName=Specular Intensity}
   )
   B=Multiply(A=Clamp(None=@n16), B=VectorParameter() {group=Color, parameterName=SpecularColor, defaultValue=<rgba 0.8093 0.0673 1 1>}.RGB) {constA=1}
  )
  NamedRerouteDeclaration(None=Add(A=Multiply(A=@n1, B=@n2), B=Add(A=@n10, B=@n17) {constB=0})) {name=Emissive, nodeColor=<rgba 0 0.6667 1 1>}
  n18 = TextureSampleParameter2D(UVs=NamedRerouteUsage()) {parameterName=Opacity Mask, samplerType=SAMPLERTYPE_LinearColor, texture=/Game/Effects/Fort_Effects/Textures/Water/T_waterSplash_D}
  n19 = ChannelMaskParameter(None=@n18.RGBA) {maskChannel=Green, parameterName=Opacity MasKChannel, defaultValue=<rgba 0 1 0 0>}
  n20 = LinearInterpolate(Alpha=Saturate(None=DynamicParameter() {defaultValue=<rgba 1 0 1 0>, paramNames [Rand 0-1, Foam, UV Distortion, Erode]}.Erode)) {constA=0.04}
  MaterialExpressionSaturate_8 = Saturate(None=Divide(A=Subtract(A=@n19.R, B=@n20), B=OneMinus(None=@n20)))
  MaterialExpressionScalarParameter_34 = ScalarParameter() {group=Opacity, parameterName=ThresholdParam}
  n21 = Custom!(Opacity=@MaterialExpressionSaturate_8, Threshold=@MaterialExpressionScalarParameter_34) {
   code="clip(Opacity - Threshold);\r\nreturn Opacity;"
   description=ClipMask
   outputType=CMOT_Float1
   inputs [
    {inputName=Opacity, input {expression=&self:MaterialExpressionSaturate_8}}
    {inputName=Threshold, input {expression=&self:MaterialExpressionScalarParameter_34}}
   ]
  }
  n22 = Saturate(None=Divide(
    A=NamedRerouteDeclaration(None=@n21) {name=Eroded Mask, nodeColor=<rgba 1 0 0.9844 1>}
    B=ScalarParameter() {defaultValue=0.1, group=Opacity, parameterName=OpacitySoftness}
   ))
  n23 = Power(
   Base=MaterialFunctionCall(Distance=ScalarParameter() {defaultValue=40, group=Opacity, parameterName=NearFade}) {materialFunction=&Near_Camera_Fade}.Result
   Exp=ScalarParameter() {defaultValue=3, group=Opacity, parameterName=NearFadeExp}
  )
  n24 = ParticleColor()
  n25 = ScalarParameter() {defaultValue=50, group=Opacity, parameterName=DepthFade}
  n26 = DepthFade(FadeDistance=LinearInterpolate(A=Multiply(A=@n25) {constB=2}, B=@n25))
  n27 = Constant() {r=1}
  n28 = NamedRerouteDeclaration(None=Multiply(A=Multiply(A=@n23, B=@n24.A), B=FeatureLevelSwitch(Default=@n26, ES2=@n27, ES3_1=@n27, SM5=@n26, SM6=@n26))) {name=Alpha Multiplier, nodeColor=<rgba 0 0.3672 1 1>}
  NamedRerouteDeclaration(None=Saturate(None=Multiply(A=@n22, B=@n28))) {name=Opacity, nodeColor=<rgba 1 1 1 1>}
  n29 = VectorParameter() {group=Color, parameterName=Light Color, defaultValue=<rgba 0.5941 0.1226 1 1>}
  n30 = Multiply(A=OneMinus(None=NamedRerouteUsage()), B=NamedRerouteUsage())
  n31 = VectorParameter() {group=Color, parameterName=Dark Color, defaultValue=<rgba 0.0322 0 0.099 1>}
  NamedRerouteDeclaration(None=Multiply(A=LinearInterpolate(A=@n29.RGB, B=@n31.RGB, Alpha=@n30), B=ParticleColor().RGB)) {name=Diffuse Color, nodeColor=<rgba 1 0.25 0 1>}
  n32 = VectorParameter() {group=Distortion, parameterName=BaseTex_UVs, channelNames {a=PanY, b=PanX, g=ScaleY, r=ScaleX}, defaultValue=<rgba 1 1 0 0>}
  n33 = Multiply(A=TextureCoordinate(), B=AppendVector(A=@n32.ScaleX, B=@n32.ScaleX))
  n34 = Multiply(A=Time(), B=AppendVector(A=@n32.PanX, B=@n32.PanX))
  n35 = VectorParameter() {group=Distortion, parameterName=Distortion_UVs, channelNames {a=PanY, b=PanX, g=ScaleY, r=ScaleX}, defaultValue=<rgba 1 1 0.05 0.02>}
  n36 = Multiply(
   A=Multiply(A=TextureCoordinate(), B=ScalarParameter() {defaultValue=8, group=Distortion, parameterName=Distortion Tile})
   B=AppendVector(A=@n35.ScaleX, B=@n35.ScaleX)
  )
  n37 = Add(
   A=Add(A=DynamicParameter() {defaultValue=<rgba 1 0 1 0>, paramNames [Rand 0-1, Foam, UV Distortion, Erode]}.Rand 0-1, B=ParticleRelativeTime())
   B=Multiply(A=Time(), B=AppendVector(A=@n35.PanX, B=@n35.PanX))
  )
  n38 = ComponentMask(None=TextureSample(UVs=Add(A=@n36, B=@n37)) {samplerType=SAMPLERTYPE_Normal, texture=&T_Weap_Distort_N_02}.RGB) {g=true, r=true}
  n39 = Multiply(
   A=ScalarParameter() {defaultValue=0.02, group=Distortion, parameterName=Distortion Amount}
   B=DynamicParameter() {defaultValue=<rgba 1 0 1 0>, paramNames [Rand 0-1, Foam, UV Distortion, Erode]}.UV Distortion
  )
  NamedRerouteDeclaration(None=Add(A=Add(A=@n33, B=@n34), B=Multiply(A=@n38, B=@n39))) {name=UVs, nodeColor=<rgba 1 1 0 1>}
  n40 = TextureSampleParameter2D(UVs=NamedRerouteUsage()) {parameterName=Normal, samplerType=SAMPLERTYPE_Normal, texture=/Game/Effects/Fort_Effects/Textures/Water/T_waterSplash_R_N}
  n41 = MaterialFunctionCall(X=@n40.R, Y=@n40.R, Z=NamedRerouteUsage()) {materialFunction=&MakeFloat3}
  NamedRerouteDeclaration(None=MaterialFunctionCall(Normal=Normalize(VectorInput=@n41.Result), Flatness=ScalarParameter() {group=Normal, parameterName=FlattenNormal}) {materialFunction=&FlattenNormal}.Result) {name=Normal, nodeColor=<rgba 0 0.25 1 1>}
  NamedRerouteUsage()
  NamedRerouteUsage()
  NamedRerouteUsage()
 ]
}
