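# Masked Niagara sprite material for loot sac goo drops, blending from a splash look to a drop look as particle speed rises.
# Normalized speed drives the texture, normal and mask blends, tinted by Color and particle color.
# Fake lighting and a tight highlight come from dotting normals with the atmospheric light vector.
# A randomized fall-and-gravity offset over particle life applies only below the speed threshold; depth fade softens alpha.
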Material /BugBiomeGameplay/Assets/VFX/M_LootSac_GooDrops
graph {
 properties {
  allowTranslucentCustomDepthWrites=true
  bUsedWithNiagaraSprites=true
  blendMode=BLEND_Masked
  pixelDepthOffsetMode=PDOM_Legacy
  usedShadingModels=Default Lit
 }
 nodes [
  n1 = VectorParameter() {parameterName=Color, defaultValue=<rgba 0.0938 0.0831 0.0754 1>}
  n2 = Saturate(None=Divide(A=ParticleSpeed(), B=ScalarParameter() {defaultValue=0.1, group=Speed, parameterName=Speed}))
  n3 = DotProduct(
   A=AtmosphericLightVector()
   B=LinearInterpolate(
    A=TextureSample() {samplerType=SAMPLERTYPE_Normal, texture=/Game/Effects/Fort_Effects/Textures/Water/T_waterSplash_N}.RGB
    B=TextureSample() {samplerType=SAMPLERTYPE_Normal, texture=/Game/Effects/Fort_Effects/Textures/Water/T_BoilBubble_Fix_N}.RGB
    Alpha=@n2
   ) {constB=0}
  )
  n4 = Power(Base=Saturate(None=Subtract(A=@n3) {constB=0.04})) {constExponent=15}
  n5 = StaticComponentMaskParameter(None=TextureSample() {samplerType=SAMPLERTYPE_LinearColor, texture=/Game/Effects/Fort_Effects/Textures/Water/T_waterSplash}.RGB) {defaultG=true, parameterName=Alpha Mask Channel}
  n6 = TextureSampleParameter2D() {parameterName=Drop Texture, texture=/Game/Effects/Fort_Effects/Textures/Water/T_BoilBubble_D}
  n7 = Multiply(A=Add(A=@n1.RGB, B=Multiply(A=@n1.RGB, B=@n4)), B=Multiply(A=LinearInterpolate(A=@n5, B=@n6.RGB, Alpha=@n2) {constB=0}, B=@n3))
  n8 = ParticleColor()
  MaterialFunctionCall(Emissive=Multiply(A=@n7, B=@n8.RGB)) {materialFunction=&MF_FX_EmissiveAdjustments}
  n9 = VectorParameter() {group=Fall, parameterName=HowFarToFallZ, defaultValue=<rgba 20 20 50 0>}
  n10 = ParticleRandom()
  n11 = Multiply(
   A=ComponentMask(None=Transform(None=@n9.RGB) {transformSourceType=TRANSFORMSOURCE_Local}) {g=true, r=true}
   B=Multiply(A=Subtract(A=@n10) {constB=0.5}) {constB=2}
  )
  n12 = ParticleRelativeTime()
  n13 = ScalarParameter() {defaultValue=2, desc=33, group=Fall, parameterName=Splash Speed Mul}
  n14 = Power(Base=Add(A=@n12), Exp=@n13)
  n15 = DynamicParameter() {paramNames [Random,Param2,Param3,Param4]}
  n16 = Multiply(
   A=Add(A=Multiply(A=@n15.Random) {constB=0.5}) {constB=0.5}
   B=VectorParameter() {group=Fall, parameterName=Gravity, defaultValue=<rgba 0 0 -1 0>}.RGB
  )
  n17 = LinearInterpolate(B=Subtract(A=AppendVector(A=@n11, B=@n9.B), B=Multiply(A=@n14, B=@n16)), Alpha=Multiply(A=@n12, B=@n13))
  n18 = ScalarParameter() {defaultValue=6, group=InitialSize, parameterName=InitialSplashSizeMin}
  n19 = ScalarParameter() {defaultValue=4, group=InitialSize, parameterName=InitialSplashSizeMax}
  n20 = ParticleRandom()
  n21 = ScalarParameter() {defaultValue=3, parameterName=SizeMaxByLife}
  n22 = Power(Base=ParticleRelativeTime(), Exp=ScalarParameter() {defaultValue=2, group=SizeScalars, parameterName=ControlSizeByLifeSpeed})
  Multiply(A=@n17, B=Multiply(A=LinearInterpolate(A=@n18, B=@n19, Alpha=@n20), B=LinearInterpolate(B=@n21, Alpha=@n22) {constA=1, constB=4}))
  n23 = LinearInterpolate(A=@n5, B=ChannelMaskParameter(None=@n6.R) {maskChannel=Alpha, parameterName=Alpha Channel, defaultValue=<rgba 0 0 0 1>}.R, Alpha=@n2) {constB=0}
  n24 = Multiply(A=@n1.A, B=@n23)
  n25 = Constant() {r=1}
  n26 = LinearInterpolate(A=ScalarParameter() {defaultValue=1, parameterName=DepthFade Max}, Alpha=@n23) {constB=0.5}
  Multiply(A=Multiply(A=@n24, B=DepthFade(Opacity=@n25, FadeDistance=@n26)), B=@n8.A)
  n27 = DynamicParameter() {paramNames [Random,Param2,Param3,Param4]}
  Multiply(A=Add(A=Multiply(A=@n27.Random) {constB=0.5}) {constB=0.5}, B=@n1.RGB)
  Multiply(A=@n3, B=@n3)
  Multiply(A=@n24, B=@n8.A)
  Multiply(B=@n7)
  n28 = Constant()
  Reroute(None=If(A=@n2, B=Constant() {r=1}, "A > B"=@n28, "A == B"=@n28, "A < B"=@n17))
 ]
}
