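# Lit material for loot sac status goo VFX, built for Niagara mesh particles with ray traced shadows off.
# A saturated, low-exponent Fresnel blends deep orange to over-bright yellow toward the rim.
# The result is tinted by particle color and scaled by the Emissive dynamic parameter, so each particle can vary.
# That value feeds the shared MF_FX_EmissiveAdjustments function as emissive.
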Material /BugBiomeGameplay/Assets/VFX/M_LootSac_StatusGoo
graph {
 properties {bAutomaticallySetUsageInEditor=false, bCastRayTracedShadows=false, bUsedWithNiagaraMeshParticles=true, usedShadingModels=Default Lit}
 nodes [
  n1 = LinearInterpolate(
   A=Constant3Vector() {constant=<rgba 1 0.0747 0.0003 1>}
   B=Constant3Vector() {constant=<rgba 2 0.9524 0 1>}
   Alpha=Saturate(None=Fresnel() {exponent=0.5})
  )
  n2 = ParticleColor()
  MaterialFunctionCall(Emissive=Multiply(A=Multiply(A=@n1, B=@n2.RGB), B=DynamicParameter() {paramNames [Emissive,Param2,Param3,Param4]}.Emissive)) {materialFunction=&MF_FX_EmissiveAdjustments}
 ]
}
