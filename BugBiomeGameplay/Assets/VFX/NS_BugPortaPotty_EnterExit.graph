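# Niagara system for the bug biome porta potty, by its name the visual effect for entering and exiting it.
# It combines four emitters: two door mesh effects (Door_FX_Mesh_A and B), an Inside_Drops emitter and a Splash emitter.
# It exposes no user variables and lists only the SystemState module, with no renderer details recorded for any emitter.
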NiagaraSystem /BugBiomeGameplay/Assets/VFX/NS_BugPortaPotty_EnterExit
dependencies {usedDataInterfaces [], usedDynamicInputs [], usedModules [&SystemState]}
emitters {
 Door_FX_Mesh_A {rendererClasses [], renderers []}
 Door_FX_Mesh_B {rendererClasses [], renderers []}
 Inside_Drops {rendererClasses [], renderers []}
 Splash {rendererClasses [], renderers []}
}
summary {
 systemName=NS_BugPortaPotty_EnterExit
 emitters [
  {emitterName=Door_FX_Mesh_A, rendererClasses []}
  {emitterName=Splash, rendererClasses []}
  {emitterName=Inside_Drops, rendererClasses []}
  {emitterName=Door_FX_Mesh_B, rendererClasses []}
 ]
 userVariables []
}
