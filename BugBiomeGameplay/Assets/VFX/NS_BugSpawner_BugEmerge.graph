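# Niagara system for the effect of a bug emerging from a bug spawner in the bug biome, judging by its name.
# Splash and drops emitters are enabled, while ash and jet spout emitters are present but disabled.
# SystemState is the only listed module, and no renderers or data interfaces appear for any of the LW-prefixed emitters.
# It exposes two boolean user parameters flagging front-end and front-end preview contexts.
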NiagaraSystem /BugBiomeGameplay/Assets/VFX/NS_BugSpawner_BugEmerge
dependencies {usedDataInterfaces [], usedDynamicInputs [], usedModules [&SystemState]}
emitters {
 LW_Ash {bEnabled=false, rendererClasses [], renderers []}
 LW_Drops {rendererClasses [], renderers []}
 LW_Jet_Spout {bEnabled=false, rendererClasses [], renderers []}
 LW_Splash {rendererClasses [], renderers []}
}
summary {
 systemName=NS_BugSpawner_BugEmerge
 emitters [
  {bEnabled=false, emitterName=LW_Ash, rendererClasses []}
  {emitterName=LW_Splash, rendererClasses []}
  {emitterName=LW_Drops, rendererClasses []}
  {bEnabled=false, emitterName=LW_Jet_Spout, rendererClasses []}
 ]
 userVariables [
  User.bIsFrontEnd {type=NiagaraBool}
  User.bIsFrontEndPreview {type=NiagaraBool}
 ]
}
