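# Niagara effect for a bug spawner's death that blasts goo and meshes out of its holes.
# Counter-indexed lookups into user position and direction arrays place and launch particles per hole, with HoleScale sizing them.
# Less_Goo emits a tapering spray of falling, dragged sprites, with FXColor-selected emitter tints and time-of-day-scaled alpha.
# MeshBlast bursts rotating, randomly purple-tinted meshes that flare bright, then fade out.
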NiagaraSystem /BugBiomeGameplay/Assets/VFX/NS_BugSpawner_Death_Holes
dependencies {
 usedDataInterfaces [
  /Script/Niagara.NiagaraDataInterfaceSimpleCounter
  &NiagaraDataInterfaceArrayInt32
  &NiagaraDataInterfaceVectorCurve
  &NiagaraDataInterfaceCurve
  &NiagaraDataInterfaceArrayColor
  &NiagaraDataInterfaceVector2DCurve
 ]
 usedDynamicInputs [&FNTimeOfDayScalar]
 usedModules [
  &SystemState
  &self:SystemSpawnScript.SystemScriptSource.SystemScriptGraph.NiagaraNodeAssignment_0.SetVariables_#1
 ]
}
emitters {
 Less_Goo {
  emitterUpdateScript [
   &EmitterState()
   &SpawnRate(
    SpawnRate=&FloatFromCurve(
     FloatCurve=DataInterface() {propertyValues {Curve {keys [
         {arriveTangent=-16.7295, interpMode=RCIM_Cubic, leaveTangent=-16.7295, tangentMode=RCTM_User, time=0, value=75}
         {arriveTangent=-564.0549, interpMode=RCIM_Cubic, leaveTangent=26.9736, tangentMode=RCTM_Break, time=0.6, value=0}
        ]}}}
     CurveIndex=$Emitter.NormalizedLoopAge
    )
    Use Spawn Probability=false
   )
   &SpawnBurst_Instantaneous(Spawn Count=15, Use Spawn Probability=false, UseLoopCountLimit=false) {enabled=false}
   /SetVariables(
    Emitter.FresnelColor:LinearColor=&SelectLinearColorFromArray(
     Array Sampling Mode=::NewEnumerator1
     Direct Array Index=$User.FXColor
     Color Selection Array=DataInterface() {propertyValues {ColorData [<rgba 1 1 1 1>, <rgba 0 0.6717 1 1>, <rgba 0.3054 0.0253 1 1>, <rgba 0.0581 0.4948 0.0239 1>]}}
     Random Seed=?
    )
    Emitter.TileColor:LinearColor=&SelectLinearColorFromArray(
     Array Sampling Mode=::NewEnumerator1
     Direct Array Index=$User.FXColor
     Color Selection Array=DataInterface() {propertyValues {ColorData [<rgba 1 1 1 1>, <rgba 0 0.6717 1 1>, <rgba 0.4341 0.0403 1 1>, <rgba 0.1127 0.6146 0 1>]}}
     Random Seed=?
    )
   ) {isSetParameters=true}
   /SetVariables(Emitter.FresnelColor:LinearColor=<rgba 0.4253 0 1 1>, Emitter.TileColor:LinearColor=<rgba 0.5385 0.0009 1 1>) {enabled=false, isSetParameters=true}
  ]
  particleSpawnScript [
   &InitializeParticle(
    Color=<rgba 1.5 1.5 1.5 1>
    Lifetime=&RandomRangeFloat(Minimum=0.4, Maximum=0.7)
    Sprite Rotation Mode=::NewEnumerator0
    Sprite Size=&Multiply_Vector2D_ByFloat(Float=$User.HoleScale, Vector2D=<1 0.3>)
    Sprite Size Mode=::NewEnumerator4
    UsePositionOffset=false
    Write Material Random=true
   )
   /SetVariables(
    Particles.Position=/Niagara/DynamicInputs/Arrays/SelectPositionFromArray(
     Array Sampling Mode:ENiagaraArraySamplingMode=::NewEnumerator1
     Direct Array Index:NiagaraInt32=/Niagara/DynamicInputs/DataInterface/GenerateIntFromCounter(SimpleCounter:NiagaraDataInterfaceSimpleCounter=DataInterface() {propertyValues {GpuSyncMode=None, InitialValue=0}})
     Interpolation:NiagaraFloat=?
     Direct Array Mode:ENiagara_WrapClamp=::NewEnumerator1
     Evaluation Type:ENiagaraRandomnessEvaluation=::NewEnumerator0
     Normalized Interpolation:NiagaraBool=true
     Recalculate Random Each Loop:NiagaraBool=false
     Position Array:NiagaraDataInterfaceArrayPosition=$User.FX_HolePositions
     Randomness Mode:ENiagaraRandomnessMode=::NewEnumerator1
     Fixed Random Seed:NiagaraBool=false
     Random Seed:NiagaraInt32=?
    )
    Particles.Velocity=&Multiply_Vector(
     A=&SelectVectorFromArray(
      Direct Array Index=&GenerateIntFromCounter(SimpleCounter:NiagaraDataInterfaceSimpleCounter=DataInterface() {propertyValues {GpuSyncMode=None, InitialValue=0}})
      Direct Array Mode=::NewEnumerator1
      Vector Selection Array=$User.FX_HoleDirections
     )
     B=&RandomRangeVector(Minimum=<20 20 200>, Maximum=<500 500 500>)
    )
   ) {isSetParameters=true}
   /SetVariables(Particles.DropColor:LinearColor=<rgba 0.5363 0.1911 1 1>) {isSetParameters=true}
  ]
  particleUpdateScript [
   &ParticleState()
   &Drag(Use Linear Drag=true, Use Rotational Drag=false)
   &DynamicMaterialParameters(
    Index 0 Param 1=&FloatFromCurve(FloatCurve=DataInterface() {propertyValues {Curve {keys [
         <at 0 0.8824 RCIM_Cubic>
         {arriveTangent=-1.6832, interpMode=RCIM_Cubic, leaveTangent=-1.6832, time=0.2027, value=-0.6108}
         {arriveTangent=2.3041, interpMode=RCIM_Cubic, leaveTangent=2.3041, tangentMode=RCTM_User, time=0.6584, value=-0.2258}
         <at 1 1>
        ]}}})
    Index 0 Param 2=&FloatFromCurve(FloatCurve=DataInterface() {propertyValues {Curve {keys [<at 0 3>,<at 0.3 1>]}}})
    Module.Param0WriteEnabled=true
    Module.Param1WriteEnabled=true
    Module.Param2WriteEnabled=true
    Module.Param3WriteEnabled=true
   )
   &ScaleSpriteSize(
    Scale Factor=&Lerp_Vector2(
     A=&Multiply_Vector2D_ByFloat(Float=2, Vector2D=&RandomRangeVector2D(Minimum=<25 275>, Maximum=<50 425>))
     B=&Multiply_Vector2D_ByFloat(Float=&RandomRangeFloat(Minimum=1, Maximum=2), Vector2D=<400 400>)
     Alpha=&FloatFromCurve(FloatCurve=DataInterface() {propertyValues {Curve {keys [
          {interpMode=RCIM_Cubic, leaveTangentWeight=1.1683, tangentMode=RCTM_User, time=0, value=0}
          {arriveTangent=0.8572, interpMode=RCIM_Cubic, leaveTangent=0.8572, tangentMode=RCTM_User, time=0.8, value=0.2}
          {arriveTangent=7.9683, arriveTangentWeight=0.9652, interpMode=RCIM_Cubic, leaveTangent=7.9683, tangentMode=RCTM_User, time=1, value=1}
         ]}}})
    )
    Scale Sprite Size Mode=::NewEnumerator2
   ) {module=ScaleSpriteSize001}
   &ScaleSpriteSize(
    Uniform Curve Index=$Emitter.Age
    Uniform Curve Scale=1.6
    Uniform Curve Sprite Scale=DataInterface() {propertyValues {Curve {keys [
        {arriveTangent=-1.2353, interpMode=RCIM_Cubic, leaveTangent=-1.2353, tangentMode=RCTM_User, time=0, value=1}
        {arriveTangent=-0.269, interpMode=RCIM_Cubic, leaveTangent=-0.269, tangentMode=RCTM_User, time=1.5, value=0}
       ]}}}
   ) {module=ScaleSpriteSize002}
   /SetVariables(Particles.DropSize:Vector2f=&Vector2DFromCurve(
     Vector2Curve=DataInterface() {propertyValues {XCurve {keys [<at 0 30>,<at 1 0>]}, YCurve {keys [<at 0 70>,<at 1 20>]}}}
     ScaleCurve=&Multiply_Vector2D_ByFloat(Float=$User.HoleScale, Vector2D=<1 1>)
    )) {isSetParameters=true}
   &ScaleSpriteSize(
    Uniform Curve Scale=$User.FXScale
    Uniform Curve Sprite Scale=DataInterface() {propertyValues {Curve {keys [
        {arriveTangent=2.4725, interpMode=RCIM_Cubic, leaveTangent=2.4725, tangentMode=RCTM_User, time=0, value=0.3}
        {arriveTangent=0.6741, interpMode=RCIM_Cubic, leaveTangent=0.6741, tangentMode=RCTM_User, time=0.2, value=0.75}
        {tangentMode=RCTM_User, time=1, value=1}
       ]}}}
   ) {enabled=false}
   &GravityForce(Gravity=&RandomRangeVector(Minimum=<0 0 -200>, Maximum=<0 0 -300>))
   &SolveForcesAndVelocity(Clamp Velocity=false, Limit Acceleration=false)
   &ScaleColor(
    Scale Alpha=$System.TOD
    Scale RGB=&VectorFromFloat(Value=&Lerp_Float(
      A=0.8
      B=0.45
      Alpha=&FloatFromCurve(
       FloatCurve=DataInterface() {propertyValues {Curve {keys [
           {interpMode=RCIM_Cubic, leaveTangentWeight=0.966, tangentMode=RCTM_User, time=0, value=0}
           {arriveTangent=1.5389, interpMode=RCIM_Cubic, leaveTangent=1.5389, tangentMode=RCTM_User, time=0.5, value=0.6}
           {interpMode=RCIM_Cubic, tangentMode=RCTM_User, time=1, value=1}
          ]}}}
       CurveIndex=&RandomRangeFloat(Minimum=0, Maximum=1)
      )
     ))
    ScaleA=true
    ScaleRGB=true
   )
   &ScaleSpriteSizeBySpeed(Max Scale Factor=<1 3>, Min Scale Factor=<1 3>, Velocity Threshold=500) {enabled=false}
   &CameraOffset(Camera Offset Amount=-15)
  ]
  rendererClasses [&NiagaraSpriteRendererProperties]
  renderers [
   {rendererClass=&NiagaraSpriteRendererProperties}
   {rendererClass=&NiagaraSpriteRendererProperties, rendererIndex=1}
  ]
 }
 MeshBlast {
  emitterSpawnScript [/SetVariables(Emitter.RandomisedColour:LinearColor=&RandomRangeLinearColor(Minimum=<rgba 0.4936 0.0198 0.8976 1>, Maximum=<rgba 0.1068 0.0032 0.1267 1>)) {isSetParameters=true}]
  emitterUpdateScript [
   &EmitterState()
   &SpawnBurst_Instantaneous(Spawn Count=6, Use Spawn Probability=false, UseLoopCountLimit=false)
  ]
  particleSpawnScript [
   &InitializeParticle(
    Color=$Emitter.RandomisedColour
    Lifetime=0.7
    Mesh Scale=&Multiply_VectorByFloat(Vector=&RandomRangeVector(Minimum=<5 3 3>, Maximum=<6 4 4>), Float=$User.HoleScale)
    Mesh Scale Mode=::NewEnumerator4
    Position Offset=<100 0 0>
    UsePositionOffset=true
    Write Material Random=true
   )
   /SetVariables(
    Particles.Position=&SelectPositionFromArray(
     Array Sampling Mode:ENiagaraArraySamplingMode=::NewEnumerator1
     Direct Array Index:NiagaraInt32=&GenerateIntFromCounter(SimpleCounter:NiagaraDataInterfaceSimpleCounter=DataInterface() {propertyValues {GpuSyncMode=None, InitialValue=0}})
     Interpolation:NiagaraFloat=?
     Direct Array Mode:ENiagara_WrapClamp=::NewEnumerator1
     Evaluation Type:ENiagaraRandomnessEvaluation=::NewEnumerator0
     Normalized Interpolation:NiagaraBool=true
     Recalculate Random Each Loop:NiagaraBool=false
     Position Array:NiagaraDataInterfaceArrayPosition=$User.FX_HolePositions
     Randomness Mode:ENiagaraRandomnessMode=::NewEnumerator1
     Fixed Random Seed:NiagaraBool=false
     Random Seed:NiagaraInt32=?
    )
    Particles.Velocity=&Multiply_Vector(
     A=&SelectVectorFromArray(
      Direct Array Index=&GenerateIntFromCounter(SimpleCounter:NiagaraDataInterfaceSimpleCounter=DataInterface() {propertyValues {GpuSyncMode=None, InitialValue=0}})
      Direct Array Mode=::NewEnumerator1
      Vector Selection Array=$User.FX_HoleDirections
     )
     B=<500 500 500>
    )
   ) {isSetParameters=true}
   /SetVariables(
    Particles.DarkEnergyScale:Vector3f=&Multiply_Vector(A=$Particles.Scale, B=&RandomRangeVector(Minimum=<0.75 1.15 1.15>, Maximum=<1 1.25 1.25>))
    Particles.VisibilityTag=&SelectIntFromArray(Direct Array Index=$Engine.ExecIndex, Int Selection Array=DataInterface() {propertyValues {IntData [0,0,1]}})
   ) {enabled=false, isSetParameters=true}
   &InitialMeshOrientation(Mesh Orientation Mode=::NewEnumerator7, Rotation=<0 0 0.5>, Use Rotation Vector=true)
  ]
  particleUpdateScript [
   &ParticleState()
   &UpdateMeshOrientation(Rotation Rate=&RandomRangeFloat(Minimum=0.3, Maximum=0.75))
   &ScaleMeshSize(Scale Factor=&VectorFromCurve(VectorCurve=DataInterface() {propertyValues {XCurve {keys [<at 0 0>,<at 0.1233 0.8452>,<at 1 0.25>]}, YCurve {keys [<at 0 1>,<at 1 1>]}, ZCurve {keys [<at -0.0135 1>,<at 1 1>]}}}))
   &ScaleColor(
    Scale Alpha=&FloatFromCurve(FloatCurve=DataInterface() {propertyValues {Curve {keys [<at 0 1.5>,<at 0.694 0.8824>,<at 1 0>]}}})
    Scale RGB=&VectorFromFloat(Value=&FloatFromCurve(FloatCurve=DataInterface() {propertyValues {Curve {keys [<at 0 12>,<at 0.4803 2.4314>,<at 1 1>]}}}))
    ScaleA=true
    ScaleRGB=true
   )
  ]
  rendererClasses [&NiagaraMeshRendererProperties]
  renderers [
   {rendererClass=&NiagaraMeshRendererProperties, rendererIndex=0}
   {rendererClass=&NiagaraMeshRendererProperties, rendererIndex=1}
  ]
 }
}
summary {
 systemName=NS_BugSpawner_Death_Holes
 emitters [
  {emitterName=MeshBlast, rendererClasses [&NiagaraMeshRendererProperties]}
  {emitterName=Less_Goo, rendererClasses [&NiagaraSpriteRendererProperties]}
 ]
 userVariables [
  User.FX_HoleDirections {
   type=NiagaraDataInterfaceArrayFloat3
   description=""
   defaultValue {
    struct=NiagaraExt_VariableValue_DataInterface
    value {
     dataInterface=&self:NiagaraDataInterfaceArrayFloat3_0
     dataInterfaceClass=&NiagaraDataInterfaceArrayFloat3
    }
   }
  }
  User.FX_HolePositions {
   type=NiagaraDataInterfaceArrayPosition
   description=""
   defaultValue {
    struct=NiagaraExt_VariableValue_DataInterface
    value {
     dataInterface=&self:NiagaraDataInterfaceArrayPosition_0
     dataInterfaceClass=&NiagaraDataInterfaceArrayPosition
    }
   }
  }
  User.FXColor {type=NiagaraInt32, defaultValue=2}
  User.HoleScale {type=NiagaraFloat, defaultValue=1, description=""}
 ]
}
