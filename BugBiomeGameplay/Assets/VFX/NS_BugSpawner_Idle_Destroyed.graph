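# Niagara system for the idle effect of a destroyed bug spawner, producing occasional dripping mesh drops.
# One distance-culled emitter spawns drops at a low rate at points picked from the User.FX_DropPositions array.
# Curves swell each drop and hold back gravity until late in life, then stretch and thin it as it falls.
# Spawn color and alpha come from a curve indexed by a time-of-day system value.
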NiagaraSystem /BugBiomeGameplay/Assets/VFX/NS_BugSpawner_Idle_Destroyed
dependencies {
 usedDataInterfaces [&NiagaraDataInterfaceColorCurve, &NiagaraDataInterfaceVectorCurve]
 usedDynamicInputs [&FNTimeOfDayScalar]
 usedModules [
  &SystemState
  &self:SystemSpawnScript.SystemScriptSource.SystemScriptGraph.NiagaraNodeAssignment_0.SetVariables_#1
 ]
}
emitters {Drops {
  emitterUpdateScript [
   &EmitterState(Enable Distance Culling=true, Module.bUseMaxDistance=true, Module.bUseMinDistance=false, Scalability Mode=::NewEnumerator1)
   &SpawnRate(SpawnRate=0.35, Use Spawn Probability=true)
  ]
  particleSpawnScript [
   &InitializeParticle(
    Color=<rgba 1 0.2315 0.8191 1>
    Lifetime Max=4
    Lifetime Min=3
    Lifetime Mode=::NewEnumerator1
    Mesh Scale Mode=::NewEnumerator3
    Sprite Size Mode=::NewEnumerator3
    Uniform Sprite Size=&Multiply_Float(A=&RandomRangeFloat(Minimum=20, Maximum=30), B=$User.FXScale)
    UsePositionOffset=false
    Write Material Random=true
   )
   &AddVelocity(Rotation Coordinate Space=::World, Velocity=<0 0 -1>)
   /SetVariables(Particles.Position=/Niagara/DynamicInputs/Arrays/SelectPositionFromArray(
     Array Sampling Mode:ENiagaraArraySamplingMode=::NewEnumerator0
     Direct Array Index:NiagaraInt32=?
     Interpolation:NiagaraFloat=?
     Direct Array Mode:ENiagara_WrapClamp=::NewEnumerator0
     Evaluation Type:ENiagaraRandomnessEvaluation=::NewEnumerator0
     Normalized Interpolation:NiagaraBool=true
     Recalculate Random Each Loop:NiagaraBool=false
     Position Array:NiagaraDataInterfaceArrayPosition=$User.FX_DropPositions
     Randomness Mode:ENiagaraRandomnessMode=::NewEnumerator0
     Fixed Random Seed:NiagaraBool=false
     Random Seed:NiagaraInt32=0
    )) {isSetParameters=true}
   &Color(Color=&ColorFromCurve(
     ColorCurve=DataInterface() {propertyValues {
       AlphaCurve {keys [<at 0 0.08 RCIM_Cubic>,<at 0.3172 0.08>,<at 0.6799 0.08>,<at 0.9764 0.6>]}
       BlueCurve {keys [<at 0 0.0972 RCIM_Cubic>,<at 0.3143 0.1319>,<at 0.6725 0.1602>,<at 0.952 0.2014 RCIM_Cubic>]}
       GreenCurve {keys [<at 0 0.0105 RCIM_Cubic>,<at 0.3143 0.0292>,<at 0.6725 0.045>,<at 0.952 0.0327 RCIM_Cubic>]}
       RedCurve {keys [<at 0 0.0863 RCIM_Cubic>,<at 0.3143 0.0976>,<at 0.6725 0.1797>,<at 0.952 0.1696 RCIM_Cubic>]}
      }}
     CurveIndex=$System.TOD
    )) {module=Color001}
  ]
  particleUpdateScript [
   &ParticleState()
   &ScaleMeshSize(Scale Factor=&VectorFromCurve(VectorCurve=DataInterface() {propertyValues {
       XCurve {keys [<at 0 0 RCIM_Cubic>, {interpMode=RCIM_Cubic, leaveTangent=-5.5242, tangentMode=RCTM_Break, time=0.8, value=1}, <at 1 0>]}
       YCurve {keys [<at 0 0 RCIM_Cubic>, {interpMode=RCIM_Cubic, leaveTangent=-6.9491, tangentMode=RCTM_Break, time=0.8, value=1}, <at 1 0>]}
       ZCurve {keys [<at 0 0>,<at 0.8 1>,<at 1 2>]}
      }}))
   &GravityForce(Gravity=&VectorFromCurve(VectorCurve=DataInterface() {propertyValues {XCurve {keys [<at 0 0>,<at 1 0>]}, YCurve {keys [<at 0 0>,<at 1 0>]}, ZCurve {keys [<at 0 0>,<at 0.8 0>,<at 0.81 -980>]}}}))
   &SolveForcesAndVelocity(Clamp Velocity=false, Limit Acceleration=false)
  ]
  rendererClasses [&NiagaraMeshRendererProperties]
  renderers [{rendererClass=&NiagaraMeshRendererProperties, rendererIndex=0}]
 }}
summary {
 systemName=NS_BugSpawner_Idle_Destroyed
 emitters [{emitterName=Drops, rendererClasses [&NiagaraMeshRendererProperties]}]
 userVariables [
  User.FX_DropPositions {
   type=NiagaraDataInterfaceArrayPosition
   description=""
   defaultValue {
    struct=NiagaraExt_VariableValue_DataInterface
    value {
     dataInterface=&self:NiagaraDataInterfaceArrayPosition_0
     dataInterfaceClass=&NiagaraDataInterfaceArrayPosition
    }
   }
  }
  User.FXColor {type=NiagaraInt32, defaultValue=2}
  User.FXScale {type=NiagaraFloat, defaultValue=1, description=""}
 ]
}
