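# Niagara status effect, apparently for a LootSac NPC, that coats a skeletal character in goo that drips and wears off.
# Two mesh emitters pin partitioned blobs to leg and upper-body bones, sized by character scale, then shrink and darken them via MudRemoveTimer.
# A GPU sprite emitter sheds drops from a skeletal mesh at a rate decaying with system age.
# Drops fall, collide without bouncing, and change size and lifetime based on speed.
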NiagaraSystem /BugBiomeGameplay/Assets/VFX/NS_LootSac_NPC_StatusGoo
dependencies {
 usedDataInterfaces [
  &NiagaraDataInterfaceCurve
  &NiagaraDataInterfaceArrayColor
  &NiagaraDataInterfaceSpriteRendererInfo
 ]
 usedDynamicInputs [
  &Lerp_Vector
  &RandomRangeFloat
  &SetBoolByIntComparison
 ]
 usedModules [
  &self:SystemSpawnScript.SystemScriptSource.SystemScriptGraph.NiagaraNodeAssignment_0.SetVariables_#1
  &self:SystemSpawnScript.SystemScriptSource.SystemScriptGraph.NiagaraNodeAssignment_2.SetVariables_#2
  &self:SystemSpawnScript.SystemScriptSource.SystemScriptGraph.NiagaraNodeAssignment_1.SetVariables_#3
  &self:SystemSpawnScript.SystemScriptSource.SystemScriptGraph.NiagaraNodeAssignment_5.SetVariables_#4
  &self:SystemSpawnScript.SystemScriptSource.SystemScriptGraph.NiagaraNodeAssignment_9.SetVariables_#5
  &self:SystemSpawnScript.SystemScriptSource.SystemScriptGraph.NiagaraNodeAssignment_6.SetVariables_#6
  &self:SystemSpawnScript.SystemScriptSource.SystemScriptGraph.NiagaraNodeAssignment_4.SetVariables_#7
  &SystemState
  &self:SetPeeloutFXArrayIndex
  &self:SystemSpawnScript.SystemScriptSource.SystemScriptGraph.NiagaraNodeAssignment_3.SetVariables_#8
 ]
}
emitters {
 GPU_Drops001 {
  simTarget=GPUComputeSim
  emitterUpdateScript [
   &EmitterState()
   &SpawnRate(SpawnRate=100, Use Spawn Probability=false) {enabled=false}
   &SpawnRate(
    SpawnGroup=1
    SpawnRate=&FloatFromCurve(FloatCurve=DataInterface() {propertyValues {Curve {keys [<at 0 5>,<at 5 0>]}}}, CurveIndex=$System.Age)
    Use Spawn Probability=false
   ) {module=SpawnRate001}
   /SetVariables(Emitter.DropsColor:LinearColor=&SelectLinearColorFromArray(
     Array Sampling Mode=::NewEnumerator1
     Direct Array Index=$User.GooColorScheme
     Color Selection Array=DataInterface() {propertyValues {ColorData [<rgba 0.015 0.0049 0.0024 1>, <rgba 0 0.6717 1 1>, <rgba 0.5672 0 1 1>, <rgba 1 0.2396 0 1>]}}
     Random Seed=?
    )) {isSetParameters=true}
  ]
  particleSpawnScript [
   &InitializeParticle(
    Color=<rgba 6 1.5872 0 5>
    Lifetime=0
    Mass Max=1.1
    Mass Min=0.9
    Mass Mode=::NewEnumerator3
    Sprite Size Mode=::NewEnumerator1
    Sprite UV Mode=::NewEnumerator3
    Uniform Sprite Size Min=6
    UsePositionOffset=false
    Write Material Random=true
   )
   &ShapeLocation(MaskSpawnGroup=false, Override Seed=false, Shape Origin=<0 0 -1>, Sphere Radius=4, UseOwnerScale=true)
   &SkeletalMeshLocation(
    Bone Sampling Mode=::NewEnumerator0
    "Inherit Velocity (Scale)"=0.0005
    MaskSpawnGroup=false
    Orientation Sampling=::NewEnumerator1
    OverrideRandomSeed=false
    Sampled Position Offset=$Particles.Initial.ShapeLocation.ShapePosition
    Sampled Position Offset Coordinate Space:ENiagaraSkelMeshTransforms=::NewEnumerator0
    Skeletal Mesh=$User.Skeletal Mesh3
    Velocity Sampling=::NewEnumerator0
    VelocitySpeedLimit=false
   )
   &ApplyInitialForces()
   /SetVariables(
    Particles.Lifetime=&MaskFloatBySpawnGroup(
     Masked Float=&Multiply_Float(
      A=$Particles.Lifetime
      B=&MakeCustomFloatFromBool(
       Bool=&SetBoolByFloatComparison(A=HlslExpression() {hlslExpression="length(Particles.SkeletalMeshLocation.SampledVelocity)"}, B=500)
       True Float=1
       False Float=0
      )
     )
     Passthrough Float=1
    )
    Particles.Stored=1
   ) {isSetParameters=true}
   &SubUVAnimation(
    "Sprite Renderer "=DataInterface() {propertyValues {SpriteRenderer=&self:GPU_Drops001_0.NiagaraSpriteRendererProperties_0}}
    SubUV Animation Mode:ENiagara_SubUVLookupModeV2=::NewEnumerator2
    UseEndFrame=false
    UseStartFrame=false
   )
  ]
  particleUpdateScript [
   &ParticleState()
   &GravityForce(Gravity=&Multiply_VectorByFloat(Vector=<0 0 -1500>, Float=1))
   &Drag(Drag=0.25, Use Linear Drag=false, Use Rotational Drag=false)
   &Collision(
    Advanced Aging Rate=0.1
    Age Colliding Particles=false
    Collision Enabled=&SetBoolByFloatComparison(
     A=1
     B=&FloatFromCurve(
      FloatCurve=DataInterface() {propertyValues {Curve {keys [<at 0.1 0>, {tangentMode=RCTM_User, time=0.26, value=1.1}]}}}
      CurveIndex=$Particles.Age
     )
     Comparison Type=::NewEnumerator6
    )
    Friction=10
    Friction During a Bounce=10000
    GPU Collision Type=::NewEnumerator2
    Maximum Penetration Correction Distance=5
    Method for Calculating Particle Radius:ENiagaraCalculateRadiusOptions=::NewEnumerator1
    Module.Control Rest State Via Surface Normal=false
    Module.EnableMaxCollisionCount=false
    Module.Randomize Collision Normal Vector=false
    Percentage of Penetration Before Rest=0
    Rest State Time Range=0.1
    Restitution=0
   )
   &ScaleColor(
    Scale Alpha=&FloatFromCurve(FloatCurve=DataInterface() {propertyValues {Curve {keys [
         <at 0.9 1.0113 RCIM_Cubic>
         {arriveTangent=-5.7525, interpMode=RCIM_Cubic, leaveTangent=-5.7525, tangentMode=RCTM_User, time=1.0011, value=0.002}
        ]}}})
    ScaleA=true
    ScaleRGB=true
   )
   &SolveForcesAndVelocity(Clamp Velocity=false, Limit Acceleration=false)
   &self:ScratchModule_01(Velo:Vector3f=$Particles.Velocity) {module=ScratchModule_01}
   /SetVariables(Particles.NewNAge=HlslExpression() {hlslExpression="(Particles.Age - Particles.Stored) * (1.0/ (1.0-Particles.Stored+0))"}) {isSetParameters=true}
   &ScaleSpriteSize(
    Scale Sprite Size Mode=::NewEnumerator0
    Uniform Scale Factor=&MakeCustomFloatFromBool(
     Bool=&SetBoolByFloatComparison(A=0.1, B=HlslExpression() {hlslExpression="length(Particles.Velocity)"})
     True Float=&FloatFromCurve(
      FloatCurve=DataInterface() {propertyValues {Curve {keys [{tangentMode=RCTM_User, time=0, value=0.1}, {tangentMode=RCTM_User, time=0.5, value=1}]}}}
      CurveIndex=$Particles.NewNAge
      Scale Curve=7
     )
     False Float=&FloatFromCurve(FloatCurve=DataInterface() {propertyValues {Curve {keys [{tangentMode=RCTM_User, time=0, value=0}, {tangentMode=RCTM_User, time=0.1, value=1}]}}})
    )
   )
   /SetVariables(
    Particles.CameraOffset:NiagaraFloat=&MakeCustomFloatFromBool(
     Bool=&SetBoolByFloatComparison(A=1, B=HlslExpression() {hlslExpression="length(Particles.Velocity)"})
     True Float=5
     False Float=1
    )
    Particles.Lifetime=&MakeCustomFloatFromBool(
     Bool=&SetBoolByFloatComparison(A=1, B=HlslExpression() {hlslExpression="length(Particles.Velocity)"})
     True Float=&Multiply_Float(A=1, B=&Add_Float(A=$Particles.Stored, B=0.25))
     False Float=$Particles.Lifetime
    )
   ) {isSetParameters=true}
   &ScaleSpriteSizeBySpeed(Max Scale Factor=<0.5 2>, Min Scale Factor=<0.5 0.6>, Velocity Threshold=500)
  ]
  rendererClasses [&NiagaraSpriteRendererProperties]
  renderers [{rendererClass=&NiagaraSpriteRendererProperties}]
 }
 Skeletal_Mesh_Particle_Legs001 {
  emitterUpdateScript [
   &EmitterState()
   &SpawnBurst_Instantaneous(Spawn Count=6, Use Spawn Probability=false, UseLoopCountLimit=true)
  ]
  particleSpawnScript [
   &PartitionParticles(Number Of Partitions=6)
   &InitialMeshOrientation(
    Mesh Axis To Orient=<0 0 1>
    Mesh Orientation Mode=::NewEnumerator2
    Orientation Vector=&SelectVectorFromArray(
     Direct Array Index=$Particles.PartitionParticles.Partition
     Vector Selection Array=DataInterface() {propertyValues {FloatData [<1 0 0>, <1 0 0>, <0 -1 0>, <-1 0 0>, <-1 0 0>, <0 1 0>]}}
    )
    Rotation=&SelectVectorFromArray(
     Direct Array Index=$Particles.PartitionParticles.Partition
     Vector Selection Array=DataInterface() {propertyValues {FloatData [<0 0 0>, <-0.02 0 0.4>, <0 0 0.5>, <0 0 0.8>, <0.06 0.3 1>, <-0.1 0 0.5>]}}
    )
    Use Rotation Vector=true
   )
   &InitializeParticle(
    Lifetime=12
    Mesh Scale=&Multiply_Vector(
     A=<0.11 0.11 0.15>
     B=&SelectVectorFromArray(
      Direct Array Index=$Particles.PartitionParticles.Partition
      Vector Selection Array=DataInterface() {propertyValues {FloatData [<0.92 0.9 0.8>, <0.7 0.8 0.8>, <0.72 0.6 0.5>, <1.3 0.85 0.75>, <0.8 0.8 0.7>, <0.8 0.8 0.58>]}}
     )
    )
    Mesh Scale Mode=::NewEnumerator4
    Sprite Rotation Mode=::NewEnumerator2
    Sprite Size Mode=::NewEnumerator1
    Uniform Sprite Size Max=6
    Uniform Sprite Size Min=2
    UsePositionOffset=false
    Write Material Random=true
   )
   &ShapeLocation(
    Apply To Particle Position=::NewEnumerator1
    Cylinder Height=1
    Cylinder Height Midpoint:NiagaraFloat=&SelectFloatFromArray(Direct Array Index=$Particles.PartitionParticles.Partition, Float Selection Array=DataInterface() {propertyValues {FloatData [0.2,0]}})
    Cylinder Radius=1
    MaskSpawnGroup=false
    Non Uniform Scale=&SelectVectorFromArray(Direct Array Index=$Particles.PartitionParticles.Partition, Vector Selection Array=$System.Shape Scale Array)
    Offset Mode=::NewEnumerator0
    Override Seed=false
    Rotation Coordinate Space=::World
    Rotation Mode=::NewEnumerator1
    Shape Origin=&VectorFromFloat(Value=0)
    Shape Primitive=::NewEnumerator1
    Surface Only=true
    Use Endcaps In Surface Only Mode=false
    UseOwnerScale=true
   )
   &ShapeLocation(
    Apply To Particle Position=::NewEnumerator1
    MaskSpawnGroup=false
    Offset=<5 0 0>
    Offset Coordinate Space=::World
    Offset Mode=::NewEnumerator0
    Override Seed=false
    Rotation Coordinate Space=::World
    Rotation Mode=::NewEnumerator1
    Shape Origin=&VectorFromFloat(Value=0)
    Sphere Radius=1
    Sphere Surface Distribution=1
    UseOwnerScale=true
   ) {module=ShapeLocation001}
   &CameraOffset(Camera Offset Amount=1) {enabled=false}
   /SetVariables(Particles.Scale=&Multiply_Vector(A=$Particles.Scale, B=$System.Adjust Scale by Character Size)) {isSetParameters=true}
  ]
  particleUpdateScript [
   &ParticleState()
   &SkeletalMeshLocation(
    Bone / Socket Index=&Add_Integer(A=11, B=$Particles.PartitionParticles.Partition)
    Bone Sampling Mode=::NewEnumerator4
    MaskSpawnGroup=false
    Sampled Position Offset=&SelectVectorFromArray(
     Direct Array Index=$Particles.PartitionParticles.Partition
     Vector Selection Array=DataInterface() {propertyValues {FloatData [<0 0 1>, <0 1 0>, <-0.5 4.5 0>, <0 0 0>, <0 0 0>, <0 -5 0>]}}
    )
    Skeletal Mesh=$User.SkeletalMesh
    Use Sampled Mesh Attribute as Ribbon ID=false
   ) {module=SkeletalMeshLocation001}
   &SpriteRotationRate(Rotation Rate=&MakeCustomFloatFromBool(Bool=&RandomBool(), True Float=-1080, False Float=1080))
   &DynamicMaterialParameters(
    Index 0 Param 1=&FloatFromCurve(FloatCurve=DataInterface() {propertyValues {Curve {keys [<at 0 0.2>,<at 0.03 0.05>,<at 1 0>]}}}, CurveIndex=$System.MudRemoveTimer)
    Module.Index0Float0Write=true
    Module.Index0Float1Write=true
    Module.Index0Float2Write=true
    Module.Index0Float3Write=true
   )
   &SolveForcesAndVelocity(Clamp Velocity=false, Limit Acceleration=false)
   /SetVariables(Particles.OriginPos:NiagaraPosition=$Particles.SkeletalMeshLocation001.SampledPosition) {isSetParameters=true, module=SetVariables_#9}
   &ScaleMeshSize(Scale Factor=&MakeVector(
     X=&MakeFloatFromVector(
      Vector=&VectorFromFloat(Value=&FloatFromCurve(FloatCurve=DataInterface() {propertyValues {Curve {keys [<at 0.1 1>,<at 1 0>]}}}, CurveIndex=$System.MudRemoveTimer))
      Channel=::NewEnumerator0
     )
     Y=&MakeFloatFromVector(
      Vector=&VectorFromFloat(Value=&FloatFromCurve(FloatCurve=DataInterface() {propertyValues {Curve {keys [<at 0.1 1>,<at 1 0>]}}}, CurveIndex=$System.MudRemoveTimer))
      Channel=::NewEnumerator0
     )
     Z=1
    ))
   &ScaleColor(
    Scale RGB=&VectorFromFloat(Value=&FloatFromCurve(FloatCurve=DataInterface() {propertyValues {Curve {keys [<at 0 1>,<at 0.0484 0.5>,<at 1 0.4>]}}}, CurveIndex=$System.MudRemoveTimer))
    ScaleA=true
    ScaleRGB=true
   )
  ]
  rendererClasses [&NiagaraMeshRendererProperties]
  renderers [{rendererClass=&NiagaraMeshRendererProperties, rendererIndex=0}]
 }
 Skeletal_Mesh_Particle_Upper_Body001 {
  emitterUpdateScript [
   &EmitterState()
   &SpawnBurst_Instantaneous(Spawn Count=8, Use Spawn Probability=false, UseLoopCountLimit=true)
  ]
  particleSpawnScript [
   &PartitionParticles(Number Of Partitions=8)
   &InitialMeshOrientation(
    Mesh Axis To Orient=<0 0 1>
    Mesh Orientation Mode=::NewEnumerator2
    Orientation Vector=&SelectVectorFromArray(
     Direct Array Index=$Particles.PartitionParticles.Partition
     Vector Selection Array=DataInterface() {propertyValues {FloatData [<-1 0 0>, <-1 0 0>, <-1 0 0.4>, <1 0 0>, <1 0 0>, <1 0 0>, <0 -0.05 -0.3>]}}
    )
    Rotation=&SelectVectorFromArray(
     Direct Array Index=$Particles.PartitionParticles.Partition
     Vector Selection Array=DataInterface() {propertyValues {FloatData [<0 0 0.4>, <0 0 0>, <0 0 0>, <0 -0.02 0.5>, <0 0 0>, <0 0 0.5>, <0 0 0.5>]}}
    )
    Use Rotation Vector=true
   )
   &InitializeParticle(
    Lifetime=12
    Mesh Scale=&Multiply_Vector(
     A=<0.06 0.07 0.09>
     B=&SelectVectorFromArray(
      Direct Array Index=$Particles.PartitionParticles.Partition
      Vector Selection Array=DataInterface() {propertyValues {FloatData [<1.5 1.2 1.02>, <1.1 1.1 0.6>, <1 1 0.5>, <1 1.2 1>, <1.1 0.8 0.6>, <0.65 0.7 0.5>, <1.2 1.4 0.6>, <1.2 1.2 0.8>]}}
     )
    )
    Mesh Scale Mode=::NewEnumerator4
    Sprite Rotation Mode=::NewEnumerator2
    Sprite Size Mode=::NewEnumerator1
    Uniform Sprite Size Max=6
    Uniform Sprite Size Min=2
    UsePositionOffset=false
    Write Material Random=true
   )
   &ShapeLocation(
    Apply To Particle Position=::NewEnumerator1
    Cylinder Height=1
    Cylinder Height Midpoint:NiagaraFloat=&SelectFloatFromArray(Direct Array Index=$Particles.PartitionParticles.Partition, Float Selection Array=DataInterface() {propertyValues {FloatData [0.2,0]}})
    Cylinder Radius=1
    MaskSpawnGroup=false
    Non Uniform Scale=&SelectVectorFromArray(Direct Array Index=$Particles.PartitionParticles.Partition, Vector Selection Array=$System.Shape Scale Array)
    Offset Mode=::NewEnumerator0
    Override Seed=false
    Rotation Coordinate Space=::World
    Rotation Mode=::NewEnumerator1
    Shape Origin=&VectorFromFloat(Value=0)
    Shape Primitive=::NewEnumerator1
    Surface Only=true
    Use Endcaps In Surface Only Mode=false
    UseOwnerScale=true
   )
   &ShapeLocation(
    Apply To Particle Position=::NewEnumerator1
    MaskSpawnGroup=false
    Offset=<5 0 0>
    Offset Coordinate Space=::World
    Offset Mode=::NewEnumerator0
    Override Seed=false
    Rotation Coordinate Space=::World
    Rotation Mode=::NewEnumerator1
    Shape Origin=&VectorFromFloat(Value=0)
    Sphere Radius=1
    Sphere Surface Distribution=1
    UseOwnerScale=true
   ) {module=ShapeLocation001}
   &CameraOffset(Camera Offset Amount=&Multiply_Float(A=1, B=&MakeFloatFromVector(Vector=$System.Adjust Scale by Character Size, Channel=::NewEnumerator0)))
   /SetVariables(Particles.Scale=&Multiply_Vector(
     A=$Particles.Scale
     B=&Multiply_Vector(A=$System.Adjust Scale by Character Size, B=$System.Adjust Scale by Character Size)
    )) {isSetParameters=true}
  ]
  particleUpdateScript [
   &ParticleState()
   &SkeletalMeshLocation(
    Bone / Socket Index=&Add_Integer(A=2, B=$Particles.PartitionParticles.Partition)
    Bone Sampling Mode=::NewEnumerator4
    MaskSpawnGroup=false
    Sampled Position Offset=&SelectVectorFromArray(
     Direct Array Index=$Particles.PartitionParticles.Partition
     Vector Selection Array=DataInterface() {propertyValues {FloatData [<4 0 -1.6>, <1 0 1>, <0 3 0>, <-3 -0.6 1.4>, <-3 0.7 -1.9>, <0.2 -0.8 -0.1>, <2 -1 5>, <5 -3 16>]}}
    )
    Skeletal Mesh=$User.SkeletalMesh
    Use Sampled Mesh Attribute as Ribbon ID=false
   ) {module=SkeletalMeshLocation001}
   &SpriteRotationRate(Rotation Rate=&MakeCustomFloatFromBool(Bool=&RandomBool(), True Float=-1080, False Float=1080))
   &DynamicMaterialParameters(
    Index 0 Param 1=&FloatFromCurve(FloatCurve=DataInterface() {propertyValues {Curve {keys [<at 0 0.2>,<at 0.03 0.05>,<at 1 0>]}}}, CurveIndex=$System.MudRemoveTimer)
    Module.Index0Float0Write=true
    Module.Index0Float1Write=true
    Module.Index0Float2Write=true
    Module.Index0Float3Write=true
   ) {module=DynamicMaterialParameters001}
   &SolveForcesAndVelocity(Clamp Velocity=false, Limit Acceleration=false)
   /SetVariables(Particles.OriginPos:NiagaraPosition=$Particles.SkeletalMeshLocation001.SampledPosition) {isSetParameters=true, module=SetVariables_#9}
   &ScaleMeshSize(Scale Factor=&MakeVector(
     X=&MakeFloatFromVector(
      Vector=&VectorFromFloat(Value=&FloatFromCurve(
        FloatCurve=DataInterface() {propertyValues {Curve {keys [<at 0.1 1 RCIM_Cubic>, {arriveTangent=-1.5873, interpMode=RCIM_Cubic, leaveTangent=-1.5873, tangentMode=RCTM_User, time=1, value=0}]}}}
        CurveIndex=$System.MudRemoveTimer
       ))
      Channel=::NewEnumerator0
     )
     Y=&MakeFloatFromVector(
      Vector=&VectorFromFloat(Value=&FloatFromCurve(
        FloatCurve=DataInterface() {propertyValues {Curve {keys [
            {arriveTangent=-0.1739, interpMode=RCIM_Cubic, leaveTangent=-0.1739, tangentMode=RCTM_User, time=0.1, value=1}
            {arriveTangent=-1.7515, interpMode=RCIM_Cubic, leaveTangent=-1.7515, tangentMode=RCTM_User, time=1, value=0}
           ]}}}
        CurveIndex=$System.MudRemoveTimer
       ))
      Channel=::NewEnumerator0
     )
     Z=1
    ))
   &ScaleMeshSize(Scale Factor=&VectorFromFloat(Value=&FloatFromCurve(
      FloatCurve=DataInterface() {propertyValues {Curve {keys [<at 0 1.1 RCIM_Cubic>, {arriveTangent=-2.8484, tangentMode=RCTM_User, time=0.9, value=0}]}}}
      CurveIndex=$System.MudRemoveTimer
     ))) {module=ScaleMeshSize001}
   &ScaleColor(
    Scale RGB=&VectorFromFloat(Value=&FloatFromCurve(FloatCurve=DataInterface() {propertyValues {Curve {keys [<at 0 1>,<at 0.0484 0.5>,<at 1 0.4>]}}}, CurveIndex=$System.MudRemoveTimer))
    ScaleA=true
    ScaleRGB=true
   )
  ]
  rendererClasses [&NiagaraMeshRendererProperties]
  renderers [{rendererClass=&NiagaraMeshRendererProperties, rendererIndex=0}]
 }
}
summary {
 systemName=NS_LootSac_NPC_StatusGoo
 emitters [
  {emitterName=GPU_Drops001, simTarget=GPUComputeSim, rendererClasses [&NiagaraSpriteRendererProperties]}
  {emitterName=Skeletal_Mesh_Particle_Legs001, rendererClasses [&NiagaraMeshRendererProperties]}
  {emitterName=Skeletal_Mesh_Particle_Upper_Body001, rendererClasses [&NiagaraMeshRendererProperties]}
 ]
 userVariables [
  User.CharacterSizeID {type=NiagaraInt32}
  User.GooColorScheme {type=NiagaraInt32, defaultValue=3}
  User.Initial Rotation {type=NiagaraFloat, defaultValue=0, description=""}
  User.Initial Rotation Range {type=NiagaraFloat, defaultValue=0, description=""}
  User.InputFromBP1 {type=Vector4f, description="", defaultValue {struct=Vector4f, value=<0 0.2 10000 2>}}
  User.InputFromBP2 {type=Vector4f, description="", defaultValue {struct=Vector4f, value=<0 0 0 0>}}
  User.Mesh Scale Array {
   type=NiagaraDataInterfaceArrayFloat3
   description=""
   defaultValue {
    struct=NiagaraExt_VariableValue_DataInterface
    value {
     dataInterface=&self:NiagaraDataInterfaceArrayFloat3_0
     dataInterfaceClass=&NiagaraDataInterfaceArrayFloat3
    }
   }
  }
  User.MudColor {type=NiagaraInt32, defaultValue=2}
  User.MudRemove {type=NiagaraInt32}
  User.MudRemoveTimer {type=NiagaraFloat, defaultValue=0.0025, description=""}
  User.Skeletal Mesh3 {
   type=NiagaraDataInterfaceSkeletalMesh
   defaultValue {
    struct=NiagaraExt_VariableValue_DataInterface
    value {dataInterface=&self:NiagaraDataInterfaceSkeletalMesh_1}
   }
  }
  User.SkeletalMesh {
   type=NiagaraDataInterfaceSkeletalMesh
   defaultValue {
    struct=NiagaraExt_VariableValue_DataInterface
    value {dataInterface=&self:NiagaraDataInterfaceSkeletalMesh_0}
   }
  }
  User.SpriteSize Array {
   type=NiagaraDataInterfaceArrayFloat
   defaultValue {
    struct=NiagaraExt_VariableValue_DataInterface
    value {
     dataInterface=&self:NiagaraDataInterfaceArrayFloat_0
     dataInterfaceClass=&NiagaraDataInterfaceArrayFloat
    }
   }
  }
 ]
}
